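# Translucent UI overlay material for the matchmaking settings UI, named as a blur overlay.
# It calls GetSlatePost0 to read the first Slate post-process buffer as the backdrop.
# A vertical gradient comes from the green channel of the normalized UI UV, shaped by an EaseOutQuint curve from MF_UI_EaseCurves and fed into a SmoothStep.
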Material /MatchmakingSettingsUI/MatchmakingSettings/Materials/M_UI_MM_Overlay_Blur
graph {
 properties {bAutomaticallySetUsageInEditor=false, blendMode=BLEND_Translucent, materialDomain=MD_UI, usedShadingModels=Default Lit}
 nodes [
  MaterialFunctionCall!("UVs (Optional)") {materialFunction=/Engine/Functions/UserInterface/GetSlatePost0}
  n1 = MaterialFunctionCall() {materialFunction=&GetUserInterfaceUV}
  SmoothStep(Value=MaterialFunctionCall("Value (0-1)"=ComponentMask(None=@n1.Normalized UV) {g=true}) {materialFunction=&MF_UI_EaseCurves}.EaseOutQuint)
 ]
}
