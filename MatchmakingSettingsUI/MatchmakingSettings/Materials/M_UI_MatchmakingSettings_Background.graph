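# Translucent UI material for the matchmaking settings screen background, built from a texture parameter scaled to fill the screen.
# The squared texture color blends by IsLoading toward a grey loading sweep, a linear gradient animated with delay and sine easing.
# An offset radial gradient with adjustable X/Y scale darkens the result to near black as a vignette.
# An optional rotatable linear mask from normalized UI UVs is toggled by Use Gradient Mask.
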Material /MatchmakingSettingsUI/MatchmakingSettings/Materials/M_UI_MatchmakingSettings_Background
graph {
 properties {bAutomaticallySetUsageInEditor=false, blendMode=BLEND_Translucent, materialDomain=MD_UI, usedShadingModels=Default Lit}
 nodes [
  n1 = TextureObjectParameter() {parameterName=Texture, texture=/Game/Global/Textures/Default/Blanks/T_Blank_White_ZeroAlpha_sRGBColor}
  n2 = TextureSample(
   UVs=MaterialFunctionCall!(ScreenUV, TextureSize=TextureProperty(Tex=@n1), ScalePivotPoint=Constant2Vector() {r=0.5}) {materialFunction=/FortUILibrary/MaterialFunctions/Basics/MF_UI_ScreenSpaceUVScaling}.Scaled to Fill
   Tex=@n1
  )
  n3 = Constant() {r=2}
  n4 = Constant() {r=0.125}
  n5 = MaterialFunctionCall(
   UVs=NamedRerouteUsage()
   GradientEnd=MaterialFunctionCall("Value (0-1)"=MaterialFunctionCall!(Animation Duration=@n3, Delay=@n4, Reset to Default) {materialFunction=/FortUILibrary/MaterialFunctions/Time/MF_UI_AnimationDelay}.Result) {materialFunction=&MF_UI_EaseCurves}.EaseInOutSine
  ) {materialFunction=&MF_UI_LinearGradient}
  n6 = OneMinus(None=Floor(None=@n5.Result))
  n7 = LinearInterpolate(
   A=Multiply(A=@n2.RGB, B=@n2.RGB)
   B=LinearInterpolate(Alpha=Multiply(A=@n5.Result, B=@n6)) {constA=0.2, constB=0.4}
   Alpha=NamedRerouteUsage()
  )
  n8 = Add(A=NamedRerouteUsage(), B=Constant2Vector() {g=0.5, r=-0.5})
  n9 = ScalarParameter() {defaultValue=0.4, parameterName=RadialScaleX}
  n10 = ScalarParameter() {defaultValue=0.4, parameterName=RadialScaleY}
  n11 = Add(A=Multiply(A=Subtract(A=@n8) {constB=0.5}, B=AppendVector(A=@n9, B=@n10)) {constB=0.5}) {constB=0.5}
  LinearInterpolate(
   A=Multiply(B=@n7) {constA=0.005, constB=0.005}
   B=Multiply(A=@n7)
   Alpha=Saturate(None=MaterialFunctionCall(UV=@n11) {materialFunction=&MF_UI_RadialGradient}.Smooth)
  )
  n12 = MaterialFunctionCall() {materialFunction=&GetUserInterfaceUV}
  n13 = ScalarParameter() {parameterName=Rotation}
  LinearInterpolate(
   B=ComponentMask(None=Rotator(Coordinate=@n12.Normalized UV, Time=@n13) {speed=6.2832}) {g=true}
   Alpha=ScalarParameter() {parameterName=Use Gradient Mask, sliderMax=1}
  ) {constA=1}
  NamedRerouteDeclaration(None=ScalarParameter() {parameterName=IsLoading, sliderMax=1}) {name=Is Loading, nodeColor=<rgba 0 0 0 1>}
  NamedRerouteDeclaration(None=ScreenPosition().ViewportUV) {name=Screen UV, nodeColor=<rgba 0 0 0 1>}
 ]
}
