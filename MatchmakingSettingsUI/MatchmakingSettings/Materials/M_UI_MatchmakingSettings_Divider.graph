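# Translucent UI-domain material for a divider in the Matchmaking Settings UI, as its name indicates.
# Its graph holds a single call to the MF_UI_CoreUI_RoundedMask function, fed a constant corner radius.
# The file shows no parameters, textures or connections from the mask to the material outputs.
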Material /MatchmakingSettingsUI/MatchmakingSettings/Materials/M_UI_MatchmakingSettings_Divider
graph {
 properties {bAutomaticallySetUsageInEditor=false, blendMode=BLEND_Translucent, materialDomain=MD_UI, usedShadingModels=Default Lit}
 nodes [MaterialFunctionCall(Corner Radius=Constant() {r=0.5}) {materialFunction=&MF_UI_CoreUI_RoundedMask}]
}
