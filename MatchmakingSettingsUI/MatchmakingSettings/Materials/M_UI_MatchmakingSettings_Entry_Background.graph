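# Translucent UI material for the background of a matchmaking settings entry, with enabled and disabled looks.
# It samples the Slate post buffer and feeds it as blur into the CoreUI BackplateB (Weak output) and BackplateDisabled functions.
# The IsDisabled scalar parameter blends between the two backplates.
# A CoreUI rounded mask function provides a corner shape with its radius set in pixels.
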Material /MatchmakingSettingsUI/MatchmakingSettings/Materials/M_UI_MatchmakingSettings_Entry_Background
graph {
 properties {bAutomaticallySetUsageInEditor=false, blendMode=BLEND_Translucent, materialDomain=MD_UI, usedShadingModels=Default Lit}
 nodes [
  n1 = MaterialFunctionCall!("UVs (Optional)") {materialFunction=/Engine/Functions/UserInterface/GetSlatePost0}
  LinearInterpolate(
   A=MaterialFunctionCall!(Post Buffer Blur=@n1.LinearRGB) {materialFunction=/FortUILibrary/MaterialFunctions/CoreUI/MF_UI_CoreUI_BackplateB}.Weak
   B=MaterialFunctionCall!(Post Buffer Blur=@n1.LinearRGB) {materialFunction=/FortUILibrary/MaterialFunctions/CoreUI/MF_UI_CoreUI_BackplateDisabled}.Result
   Alpha=ScalarParameter() {parameterName=IsDisabled, sliderMax=1}
  )
  MaterialFunctionCall(Corner Radius=Constant() {r=10}, Radius in Pixels=Constant() {r=1}) {materialFunction=&MF_UI_CoreUI_RoundedMask}
 ]
}
