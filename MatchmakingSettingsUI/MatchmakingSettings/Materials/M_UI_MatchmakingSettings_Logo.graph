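# Translucent UI material for a logo in the matchmaking settings UI, with the image supplied through a Logo texture parameter.
# MF_UI_UVScaling takes the texture's size and provides "Scaled to Fit" UVs, which sample the logo with clamped addressing.
# Those same UVs and the sampled alpha feed MF_UI_ClipEdgePixel, which by its name trims edge pixels left by clamping outside the fitted image.
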Material /MatchmakingSettingsUI/MatchmakingSettings/Materials/M_UI_MatchmakingSettings_Logo
graph {
 properties {bAutomaticallySetUsageInEditor=false, blendMode=BLEND_Translucent, materialDomain=MD_UI, usedShadingModels=Default Lit}
 nodes [
  n1 = TextureObjectParameter() {parameterName=Logo}
  n2 = MaterialFunctionCall(TextureSize=TextureProperty(Tex=@n1), ScalePivotPoint=Constant2Vector()) {materialFunction=&MF_UI_UVScaling}
  MaterialFunctionCall!(UVs=@n2.Scaled to Fit, Alpha=TextureSample(UVs=@n2.Scaled to Fit, Tex=@n1) {samplerSource=SSM_Clamp_WorldGroupSettings}.A) {materialFunction=/FortUILibrary/MaterialFunctions/Basics/MF_UI_ClipEdgePixel}
 ]
}
