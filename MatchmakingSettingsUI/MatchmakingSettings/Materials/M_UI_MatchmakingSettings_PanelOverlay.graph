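# Translucent UI material for the matchmaking settings panel overlay, built on the Slate post buffer.
# A screen-space radial gradient forms the base mask, shaped by a rounded-box mask with an outline.
# The post buffer color blends toward a weak CoreUI backplate treatment through a rerouted mask.
# "Use Gradient Mask" swaps the rounded shape and backplate blend for a linear UV fade, turned by "Rotation", over plain post-buffer color.
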Material /MatchmakingSettingsUI/MatchmakingSettings/Materials/M_UI_MatchmakingSettings_PanelOverlay
graph {
 properties {bAutomaticallySetUsageInEditor=false, blendMode=BLEND_Translucent, materialDomain=MD_UI, usedShadingModels=Default Lit}
 nodes [
  n1 = Constant()
  n2 = Constant()
  n3 = Add(A=ScreenPosition().ViewportUV, B=Constant2Vector() {g=0.5, r=-0.5})
  n4 = Add(A=Multiply(A=Subtract(A=@n3) {constB=0.5}) {constB=0.25}) {constB=0.5}
  n5 = MaterialFunctionCall(UV=Add(
    A=AppendVector(A=@n1, B=@n2)
    B=MaterialFunctionCall!(ScreenUV=@n4, TextureSize, ScalePivotPoint) {materialFunction=/FortUILibrary/MaterialFunctions/Basics/MF_UI_ScreenSpaceUVScaling}.ScaleY
   )) {materialFunction=&MF_UI_RadialGradient}
  n6 = Rotator(
   Coordinate=MaterialFunctionCall() {materialFunction=&GetUserInterfaceUV}.Normalized UV
   Time=ScalarParameter() {parameterName=Rotation}
  ) {speed=6.2832}
  n7 = ScalarParameter() {parameterName=Use Gradient Mask, sliderMax=1}
  n8 = Multiply(A=Saturate(None=Multiply(A=@n5.Smooth) {constB=1.75}), B=LinearInterpolate(B=ComponentMask(None=@n6) {g=true}, Alpha=@n7) {constA=1})
  n9 = MaterialFunctionCall(Corner Radius=Constant() {r=10}, Radius in Pixels=Constant() {r=1}, Outline Thickness=Constant() {r=3}) {materialFunction=&MF_UI_CoreUI_RoundedMask}
  LinearInterpolate(
   A=LinearInterpolate(A=Multiply(A=@n8, B=@n9.Shape Mask), Alpha=NamedRerouteDeclaration(None=@n9.Outline Mask) {name=Outline Mask, nodeColor=<rgba 0 0 0 1>})
   B=@n8
   Alpha=@n7
  )
  n10 = MaterialFunctionCall!("UVs (Optional)") {materialFunction=/Engine/Functions/UserInterface/GetSlatePost0}
  LinearInterpolate(
   A=LinearInterpolate(
    A=@n10.LinearRGB
    B=MaterialFunctionCall!(Post Buffer Blur=@n10.LinearRGB) {materialFunction=/FortUILibrary/MaterialFunctions/CoreUI/MF_UI_CoreUI_BackplateB}.Weak
    Alpha=NamedRerouteUsage()
   )
   B=@n10.LinearRGB
   Alpha=@n7
  )
  MaterialFunctionCall() {materialFunction=&GetUserInterfaceUV}
 ]
}
