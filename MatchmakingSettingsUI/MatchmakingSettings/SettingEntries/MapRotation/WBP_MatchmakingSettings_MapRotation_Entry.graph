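# Matchmaking settings entry widget that shows current and next map rotation side by side.
# Two equal-width MapRotationWidget children host a "Current Map" label and an icon countdown badge in their named slots.
# HandleNextMapRotationDateSet passes the next rotation date to a FortCountdownTimerVM view model with empty live text, then starts its timer.
# The event graph is empty, and the function's caller is not shown here.
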WidgetBlueprint /MatchmakingSettingsUI/MatchmakingSettings/SettingEntries/MapRotation/WBP_MatchmakingSettings_MapRotation_Entry
graphs [
 EventGraph {links [], nodes []}
 HandleNextMapRotationDateSet {nodes [
   n1 = HandleNextMapRotationDateSet!() -> (then, DateTime:Date Time Structure) {is=title K2Node_FunctionEntry}
   n2 = Get FortCountdownTimerVM() -> (FortCountdownTimerVM:Countdown Timer View Model Object Reference)
   > SetupViewModel!(
    execute
    self:Countdown Timer View Model Object Reference=@n2
    InFutureTime:"Date Time Structure (by ref)"=@n1
    InLiveText:"Text (by ref)"=GetEmptyText!() -> (ReturnValue:Text)
   ) -> (then)
   > StartTimer(self:Countdown Timer View Model Object Reference=@n2)
  ]}
]
variables []
widgets {
 info {inheritedWidgetCount=0, namedSlotCount=2, parentClass=&UserWidget, rootWidgetClass=&HorizontalBox, widgetCount=6}
 widgets [
  HB_Maps {widgetClassPath=&HorizontalBox, widget=HB_Maps}
  MapRotationWidget_Current {
   widgetClassPath=&/WBP_MatchmakingSettings_MapRotationWidget
   bIsVariable=true
   parent=HB_Maps
   widget=MapRotationWidget_Current
   slot {class=HorizontalBoxSlot, properties {size {sizeRule=Fill, value=1}}}
  }
  Txt_MapStatus {
   widgetClassPath=&CommonTextBlock
   namedSlotHost=MapRotationWidget_Current
   widget=Txt_MapStatus
   properties {
    autoWrapText=true
    renderOpacity=0.7
    text=Current Map
    font {
     bForceMonospaced=false
     bMaterialIsStencil=false
     fontMaterial=null
     fontObject=&NotoSans
     letterSpacing=0
     monospacedWidth=1
     size=24
     skewAmount=0
     typefaceFontName=NotoSans-Regular
     outlineSettings {
      bApplyOutlineToDropShadows=false
      bMiteredCorners=false
      bSeparateFillAlpha=false
      outlineMaterial=null
      outlineSize=0
      outlineColor=<rgba 0 0 0 1>
     }
    }
    shadowOffset=<0 0>
   }
  }
  MapRotationWidget_Next {
   widgetClassPath=&WBP_MatchmakingSettings_MapRotationWidget
   bIsVariable=true
   parent=HB_Maps
   widget=MapRotationWidget_Next
   slot {class=HorizontalBoxSlot, properties {padding=<ltrb 12 0 0 0>, size {sizeRule=Fill, value=1}}}
  }
  O_Timer {widgetClassPath=&Overlay, namedSlotHost=MapRotationWidget_Next, widget=O_Timer}
  MapRotationCountdown {
   widgetClassPath=/FortUILibrary/UIKit/Badge/WBP_UIKit_Badge
   bIsVariable=true
   parent=O_Timer
   widget=MapRotationCountdown
   properties {
    badgeText=$MapRotationTimer
    badgeType=Text
    indicator Type=Utility
    indicatorConfig=Icon
    indicatorIcon=/FortUILibrary/Textures/MaterialInstances/MI_UIKit_Icon_UpdatePending
    pixelSnapping=Disabled
    visibility=HitTestInvisible
   }
   slot {class=OverlaySlot}
  }
 ]
}
