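# Matchmaking settings panel widget that lays out a setting's title and description, likely as a details view.
# A vertical box stacks both texts above a divider and a mode info row with Playing Now and Avg Duration badges.
# The divider and mode info row start collapsed.
# The event graph holds only unconnected PreConstruct, Construct and Tick events, so the widget itself has no logic to fill its placeholder texts.
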WidgetBlueprint /MatchmakingSettingsUI/MatchmakingSettings/WBP_MatchmakingSettings_Panel
graphs [EventGraph {nodes [
   Event PreConstruct()
   Event Construct()
   Event Tick()
  ]}]
variables []
widgets {
 info {inheritedWidgetCount=0, parentClass=&UserWidget, rootWidgetClass=&VerticalBox, widgetCount=11}
 widgets [
  VerticalBox_25 {widgetClassPath=&VerticalBox, widget=VerticalBox_25}
  Txt_SettingTitle {
   widgetClassPath=&CommonTextBlock
   bIsVariable=true
   parent=VerticalBox_25
   widget=Txt_SettingTitle
   properties {
    autoWrapText=true
    text=$Setting Title
    font {
     bForceMonospaced=false
     bMaterialIsStencil=false
     fontMaterial=null
     fontObject=&NotoSans
     letterSpacing=0
     monospacedWidth=1
     size=24
     skewAmount=0
     typefaceFontName=NotoSans-Regular
     outlineSettings {
      bApplyOutlineToDropShadows=false
      bMiteredCorners=false
      bSeparateFillAlpha=false
      outlineMaterial=null
      outlineSize=0
      outlineColor=<rgba 0 0 0 1>
     }
    }
    shadowOffset=<0 0>
   }
   slot {class=VerticalBoxSlot}
  }
  Txt_SettingDesc {
   widgetClassPath=&CommonTextBlock
   bIsVariable=true
   parent=VerticalBox_25
   widget=Txt_SettingDesc
   properties {
    autoWrapText=true
    renderOpacity=0.7
    text=$Setting Description
    font {
     bForceMonospaced=false
     bMaterialIsStencil=false
     fontMaterial=null
     fontObject=&NotoSans
     letterSpacing=0
     monospacedWidth=1
     size=24
     skewAmount=0
     typefaceFontName=NotoSans-Regular
     outlineSettings {
      bApplyOutlineToDropShadows=false
      bMiteredCorners=false
      bSeparateFillAlpha=false
      outlineMaterial=null
      outlineSize=0
      outlineColor=<rgba 0 0 0 1>
     }
    }
    shadowOffset=<0 0>
   }
   slot {class=VerticalBoxSlot, properties {padding=<ltrb 0 8 0 0>, size {sizeRule=Fill, value=1}}}
  }
  Img_Divider {
   widgetClassPath=&Image
   bIsVariable=true
   parent=VerticalBox_25
   widget=Img_Divider
   properties {
    renderOpacity=0.2
    visibility=Collapsed
    brush {
     bIsDynamicallyLoaded=false
     drawAs=Image
     imageType=FullColor
     mirroring=NoMirror
     resourceName=None
     resourceObject=/MatchmakingSettingsUI/MatchmakingSettings/MaterialInstances/MI_UI_MatchmakingSettings_Divider
     tiling=NoTile
     imageSize=<32 2>
     margin=<ltrb 0 0 0 0>
     outlineSettings {
      bUseBrushTransparency=false
      roundingType=HalfHeightRadius
      width=0
      color {colorUseRule=UseColor_Specified, specifiedColor=<rgba 0 0 0 0>}
      cornerRadii=<0 0 0 0>
     }
     tintColor {colorUseRule=UseColor_Specified, specifiedColor=<rgba 1 1 1 1>}
     uVRegion {bIsValid=false, max=<0 0>, min=<0 0>}
    }
   }
   slot {class=VerticalBoxSlot, properties {padding=<ltrb 0 20 0 20>}}
  }
  HB_ModeInfo {
   widgetClassPath=&HorizontalBox
   parent=VerticalBox_25
   widget=HB_ModeInfo
   properties {visibility=Collapsed}
   slot {class=VerticalBoxSlot}
  }
  VB_CCU {widgetClassPath=&VerticalBox, parent=HB_ModeInfo, widget=VB_CCU, slot {class=HorizontalBoxSlot}}
  Txt_PlayingNow {
   widgetClassPath=&CommonTextBlock
   parent=VB_CCU
   widget=Txt_PlayingNow
   properties {
    text=Playing Now
    colorAndOpacity {colorUseRule=UseColor_Specified, specifiedColor=<rgba 0.5776 0.5776 0.5776 1>}
    font {
     bForceMonospaced=false
     bMaterialIsStencil=false
     fontMaterial=null
     fontObject=&NotoSans
     letterSpacing=0
     monospacedWidth=1
     size=24
     skewAmount=0
     typefaceFontName=NotoSans-Regular
     outlineSettings {
      bApplyOutlineToDropShadows=false
      bMiteredCorners=false
      bSeparateFillAlpha=false
      outlineMaterial=null
      outlineSize=0
      outlineColor=<rgba 0 0 0 1>
     }
    }
    shadowOffset=<0 0>
   }
   slot {class=VerticalBoxSlot, properties {horizontalAlignment=HAlign_Left, verticalAlignment=VAlign_Top}}
  }
  Badge_CCU {
   widgetClassPath=/FortUILibrary/UIKit/Badge/WBP_UIKit_Badge
   bIsVariable=true
   parent=VB_CCU
   widget=Badge_CCU
   properties {badgeSize=Default, badgeText=$CCU, badgeType=Text, indicatorConfig=None}
   slot {class=VerticalBoxSlot, properties {horizontalAlignment=HAlign_Left, verticalAlignment=VAlign_Top, padding=<ltrb 0 4 0 0>}}
  }
  VB_Duration {
   widgetClassPath=&VerticalBox
   parent=HB_ModeInfo
   widget=VB_Duration
   slot {class=HorizontalBoxSlot, properties {padding=<ltrb 40 0 0 0>}}
  }
  Txt_Duration {
   widgetClassPath=&CommonTextBlock
   parent=VB_Duration
   widget=Txt_Duration
   properties {
    text=Avg Duration
    colorAndOpacity {colorUseRule=UseColor_Specified, specifiedColor=<rgba 0.5776 0.5776 0.5776 1>}
    font {
     bForceMonospaced=false
     bMaterialIsStencil=false
     fontMaterial=null
     fontObject=&NotoSans
     letterSpacing=0
     monospacedWidth=1
     size=24
     skewAmount=0
     typefaceFontName=NotoSans-Regular
     outlineSettings {
      bApplyOutlineToDropShadows=false
      bMiteredCorners=false
      bSeparateFillAlpha=false
      outlineMaterial=null
      outlineSize=0
      outlineColor=<rgba 0 0 0 1>
     }
    }
    shadowOffset=<0 0>
   }
   slot {class=VerticalBoxSlot, properties {horizontalAlignment=HAlign_Left, verticalAlignment=VAlign_Top}}
  }
  Badge_Duration {
   widgetClassPath=&WBP_UIKit_Badge
   bIsVariable=true
   parent=VB_Duration
   widget=Badge_Duration
   properties {badgeSize=Default, badgeText=$Time, badgeType=Text, indicatorConfig=None}
   slot {class=VerticalBoxSlot, properties {horizontalAlignment=HAlign_Left, verticalAlignment=VAlign_Top, padding=<ltrb 0 4 0 0>}}
  }
 ]
}
