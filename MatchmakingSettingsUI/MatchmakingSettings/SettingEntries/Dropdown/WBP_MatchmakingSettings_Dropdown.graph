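# Matchmaking settings list entry that presents a discrete enum setting as a context-menu dropdown.
# On list item assignment, it stores the setting view model and builds one menu action per option, encoding the array index as the InteractionNameID.
# A shared pressed callback decodes that index and sets the value, while HandleOptionValueChanged updates the button text.
# Menu open and close animate the anchor button, and compact UI raises its minimum height.
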WidgetBlueprint /MatchmakingSettingsUI/MatchmakingSettings/SettingEntries/Dropdown/WBP_MatchmakingSettings_Dropdown
graphs [
 EventGraph {nodes [
   "On Context Menu Opened (Dropdown)"() {is=title K2Node_ComponentBoundEvent}
   > n1 = Cast To WBP_MatchmakingSettings_DropdownButton!(
    execute
    Object=Get ContextMenuButton!(self:WBP UIKit Context Menu Anchor Button Base Object Reference=Get Dropdown() -> (Dropdown:WBP UIKit Context Menu Anchor Button Object Reference)) -> (ContextMenuButton:WBP UIKit Button Generic Base Object Reference)
   ) -> (then, CastFailed, AsWBP Matchmaking Settings Dropdown Button:WBP Matchmaking Settings Dropdown Button Object Reference) {is=title K2Node_DynamicCast}
   > PlayAnimationForward(
    self=@n1
    InAnimation=Get Anim_ContextOpen!(self:WBP Matchmaking Settings Dropdown Button Object Reference=@n1) -> (Anim_ContextOpen:Widget Animation Object Reference)
   )
   "On Context Menu Closed (Dropdown)"() {is=title K2Node_ComponentBoundEvent}
   > n2 = Cast To WBP_MatchmakingSettings_DropdownButton(Object=Get ContextMenuButton(self=Get Dropdown()))
   > PlayAnimationReverse(self=@n2, InAnimation=Get Anim_ContextOpen(self=@n2))
   n3 = CustomEvent!() -> (OutputDelegate, then, InteractionNameID:Name) {is=title K2Node_CustomEvent}
   > n4 = Get!(execute:Exec) -> (then:Exec, else:Exec, SettingVM:Fort Matchmaking Setting Discrete Enum VM Object Reference)
   n5 = String To Integer(InString="To String (Name)"(InName=@n3.InteractionNameID))
   n6 = GetOptions!(self:Fort Matchmaking Setting Discrete Enum VM Object Reference=@n4) -> (ReturnValue:Array of Matchmaking Setting Option Structures)
   > Branch(Condition=Is Valid Index!(TargetArray:Array of Matchmaking Setting Option Structures=@n6, IndexToTest:int=@n5) -> (ReturnValue:bool))
   > SetValue!(
    execute
    self:Fort Matchmaking Setting Discrete Enum VM Object Reference=@n4
    Value:String="Get (a copy)"!(Array:Array of Matchmaking Setting Option Structures=@n6, Dimension 1=@n5) -> (Output_bIsEnabled:bool=~, Output_Value:String, Output_Name:Text).Output_Value
    LocalPlayer:Local Player Object Reference=GetOwningLocalPlayer!(self:Widget Object Reference) -> (ReturnValue:Local Player Object Reference)
   ) -> (then)
   n7 = Event OnListItemObjectSet()
   > n8 = Cast To FortMatchmakingSettingDiscreteEnumVM!(execute, Object=@n7.ListItemObject) -> (then, CastFailed, AsFort Matchmaking Setting Discrete Enum VM:Fort Matchmaking Setting Discrete Enum VM Object Reference) {is=title K2Node_DynamicCast}
   > Set SettingVM(SettingVM:Fort Matchmaking Setting Discrete Enum VM Object Reference=@n8) -> (Output_Get:Fort Matchmaking Setting Discrete Enum VM Object Reference)
   > Clear(TargetArray:Array of UIKit Context Menu Action Wrapper Object References=Get Actions() -> (Actions:Array of UIKit Context Menu Action Wrapper Object References))
   > n9 = For Each Loop(Array:Array of Matchmaking Setting Option Structures=GetOptions(self=@n8)) -> (Array Element:Matchmaking Setting Option Structure)
   n10 = Get UIKitContextMenuViewModel() -> (UIKitContextMenuViewModel:UIKit Context Menu Viewmodel Object Reference)
   Set Actions!(
    execute=@n9.Completed
    Actions:Array of UIKit Context Menu Action Wrapper Object References=Get Actions()
    self:UIKit Context Menu Viewmodel Object Reference=@n10
   ) -> (then, Output_Get:Array of UIKit Context Menu Action Wrapper Object References)
   > Set ViewModel!(
    execute
    ViewModel:UIKit Context Menu Viewmodel Object Reference=@n10
    self:WBP UIKit Context Menu Anchor Button Base Object Reference=Get Dropdown()
   ) -> (then, Output_Get:UIKit Context Menu Viewmodel Object Reference)
   n11 = Construct Object from Class(execute=@n9.LoopBody, Class=/Script/FortniteUI.UIKitContextMenuActionWrapper) -> (ReturnValue:UIKit Context Menu Action Wrapper Object Reference)
   n12 = Break Matchmaking Setting Option!(MatchmakingSettingOption:"Matchmaking Setting Option Structure (by ref)"=@n9) -> (bIsEnabled:bool=~, Value:String, Name:Text) {is=title K2Node_BreakStruct}
   n13 = "To String (Integer)"(InInt=@n9.Array Index)
   > SetActionInfo!(
    execute
    self:UIKit Context Menu Action Wrapper Object Reference=@n11
    InActionInfo:"UIKit Context Menu Action Info Structure (by ref)"=Make UIKit Context Menu Action Info!(
     DisplayName:Text=@n12.Name
     IconTexture:Texture 2D Soft Object Reference
     IconMaterial:Material Instance Soft Object Reference
     SortPriority:int=1000
     bIsAvailable:bool=True
     bIsEnabled:bool=@n12
     bIsSelected:bool=False
     bAddDivider:bool=False
     bPositionDividerAfter:bool=False
     bCloseMenuOnSelect:bool=true
     InteractionNameID:Name=String To Name(InString=@n13)
     Interaction:Data Table Row Handle Structure=(DataTable=None,RowName="")
     PressedDelegate:Delegate=@n3
     ReleasedDelegate:Delegate="(null).None"
    ) -> (UIKitContextMenuActionInfo:UIKit Context Menu Action Info Structure) {is=title K2Node_MakeStruct}
   ) -> (then)
   > Add(
    TargetArray:Array of UIKit Context Menu Action Wrapper Object References=Get Actions()
    NewItem:"UIKit Context Menu Action Wrapper Object Reference (by ref)"=@n11
   ) {is=title K2Node_CallArrayFunction}
   Event On Clicked()
   > ToggleContextMenu(self:WBP UIKit Context Menu Anchor Button Base Object Reference=Get Dropdown())
  ]}
 HandleOptionValueChanged {nodes [
   n1 = HandleOptionValueChanged!() -> (then, Value:String) {is=title K2Node_FunctionEntry}
   n2 = GetOptionForValue!(
    self:Fort Matchmaking Setting Discrete Enum VM Object Reference=Get SettingVM() -> (SettingVM:Fort Matchmaking Setting Discrete Enum VM Object Reference)
    Value:String=@n1
   ) -> (Option:Matchmaking Setting Option Structure, bFoundValidOption:bool)
   > Branch(Condition=@n2.bFoundValidOption)
   > UpdateText!(
    execute
    self:WBP UIKit Context Menu Anchor Button Base Object Reference=Get Dropdown()
    Text:Text=Break Matchmaking Setting Option(MatchmakingSettingOption=@n2) -> (bIsEnabled=~).Name
   ) -> (then)
  ]}
 SetMobileStyle {nodes [
   SetMobileStyle()
   > Parent: SetMobileStyle() {is=title K2Node_CallParentFunction}
   > Set Min Height!(
    execute
    Min Height:int=Select(Option 0:int=0, Option 1:int=64, Index=IsUsingCompactUI()) -> (ReturnValue:int)
    self:WBP UIKit Context Menu Anchor Button Base Object Reference=Get Dropdown()
   ) -> (then, Output_Get:int=0)
  ]}
]
variables [Actions {type="array<UIKitContextMenuActionWrapper>", default []}]
widgets {
 info {
  inheritedWidgetCount=0
  namedSlotCount=1
  parentClass=/MatchmakingSettingsUI/MatchmakingSettings/SettingEntries/WBP_MatchmakingSettings_EntryBase.WBP_MatchmakingSettings_EntryBase_C
  rootWidgetClass=null
  widgetCount=1
 }
 widgets [Dropdown {
   widgetClassPath=/FortUILibrary/UIKit/ContextMenu/WBP_UIKit_ContextMenuAnchorButton
   bIsVariable=true
   widget=Dropdown
   properties {
    button Background Material=null
    button Outline Material=null
    buttonClass=&/WBP_MatchmakingSettings_DropdownButton.WBP_MatchmakingSettings_DropdownButton_C
    buttonIcon=/FortUILibrary/Textures/Icons/T_UI_Icon_UIKit_Chevron_MSDF
    menuPlacement=MenuPlacement_AboveRightAnchor
    button Padding=<ltrb 0 0 0 0>
    button Text Font {
     bForceMonospaced=false
     bMaterialIsStencil=false
     fontMaterial=null
     fontObject=&HeadingNow
     letterSpacing=0
     monospacedWidth=1
     size=15.75
     skewAmount=0
     typefaceFontName=74 Regular
     outlineSettings {
      bApplyOutlineToDropShadows=false
      bMiteredCorners=false
      bSeparateFillAlpha=false
      outlineMaterial=null
      outlineSize=0
      outlineColor=<rgba 0 0 0 1>
     }
    }
   }
  }]
}
