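# Matchmaking settings entry widget that lists a discrete setting's options as radio-button rows in a CommonListView.
# GetDiscreteOptionsListItems wraps each option text in a UIKitButtonViewModel for the list.
# Setting the list item object caches the discrete setting view model and selects its current index; selection writes back the option index found by label text.
# All items report as selectable or navigable; the current-option comparison in that function is unconnected.
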WidgetBlueprint /MatchmakingSettingsUI/MatchmakingSettings/SettingEntries/EnumSelector/WBP_MatchmakingSettings_EnumSelector
graphs [
 EventGraph {nodes [
   n1 = "On Item Selection Changed (ListView)"!() -> (OutputDelegate, then, Item:Object Reference, bIsSelected:bool) {is=title K2Node_ComponentBoundEvent}
   > Branch(Condition=@n1.bIsSelected)
   > n2 = Cast To UIKitButtonViewModel!(execute, Object=@n1.Item) -> (then, CastFailed, AsUI Kit Button View Model:UI Kit Button View Model Object Reference) {is=title K2Node_DynamicCast}
   > n3 = Cast To FortMatchmakingSettingDiscreteVM!(execute, Object=GetListItemObject!(UserObjectListEntry:User Object List Entry Interface) -> (ReturnValue:Object Reference)) -> (then, CastFailed, AsFort Matchmaking Setting Discrete VM:Fort Matchmaking Setting Discrete VM Object Reference) {is=title K2Node_DynamicCast}
   > SetDiscreteOptionByIndex!(
    execute
    self:Fort Setting Discrete VM Object Reference=@n3
    Index:int=Find Item!(
     TargetArray:Array of Texts=GetDiscreteOptions!(self:Fort Setting Discrete VM Object Reference=@n3) -> (ReturnValue:Array of Texts)
     ItemToFind:"Text (by ref)"=Get Label!(self:UI Kit Button View Model Object Reference=@n2) -> (Label:Text)
    ) -> (ReturnValue)
   ) -> (then)
   n4 = Event OnListItemObjectSet()
   > n5 = Cast To FortMatchmakingSettingDiscreteVM(Object=@n4.ListItemObject)
   > Set FortMatchmakingSettingDiscreteVM(FortMatchmakingSettingDiscreteVM:Fort Matchmaking Setting Discrete VM Object Reference=@n5) -> (Output_Get:Fort Matchmaking Setting Discrete VM Object Reference)
   > SetSelectedIndex!(
    execute
    self:List View Object Reference=Get ListView() -> (ListView:Common List View Object Reference)
    Index:int=GetDiscreteOptionIndex!(self:Fort Setting Discrete VM Object Reference=@n5) -> (ReturnValue:int=0)
   ) -> (then)
   PrintText!(execute=@n5.CastFailed, InText="LOCTEXT(\"Wrong Setting!\")", bPrintToScreen, bPrintToLog, TextColor=(R=1,G=0,B=0.2333,A=1), Duration=5, Key) -> (then)
  ]}
 GetDiscreteOptionsListItems {nodes [
   n1 = GetDiscreteOptionsListItems!() -> (then, Options:Array of Texts) {is=title K2Node_FunctionEntry}
   > n2 = For Each Loop(Array:Array of Texts=@n1) -> (Array Element:Text)
   Return Node!(execute=@n2.Completed, Items:Array of Object References=Get Data() -> (Data:Array of Object References))
   > n3 = Construct Object from Class(Class=/Script/FortniteUI.UIKitButtonViewModel) -> (ReturnValue:UI Kit Button View Model Object Reference)
   > Set Label!(execute, Label:Text=@n2, self:UI Kit Button View Model Object Reference=@n3) -> (then, Output_Get:Text)
   > Add(TargetArray:Array of Object References=Get Data(), NewItem:"Object Reference (by ref)"=@n3) {is=title K2Node_CallArrayFunction}
  ]}
 GetItemSelectableOrNavigable {nodes [
   n1 = GetItemSelectableOrNavigable!() -> (then, Item:Object Reference) {is=title K2Node_FunctionEntry}
   > Return Node!(execute, ReturnValue:bool=true)
   n2 = Get FortMatchmakingSettingDiscreteVM() -> (FortMatchmakingSettingDiscreteVM:Fort Matchmaking Setting Discrete VM Object Reference)
   n3 = Cast To UIKitButtonViewModel(Object=@n1)
   "Equal (Integer)"(A=Find Item(TargetArray=GetDiscreteOptions(self=@n2), ItemToFind=Get Label(self=@n3)), B=GetDiscreteOptionIndex(self=@n2) -> (ReturnValue=0))
  ]}
]
variables []
widgets {
 info {inheritedWidgetCount=1, parentClass=/Script/FortniteUI.FortPolymorphicListView, rootWidgetClass=&Overlay, widgetCount=3}
 widgets [
  Overlay_27 {widgetClassPath=&Overlay, widget=Overlay_27}
  Background {
   widgetClassPath=&Image
   parent=Overlay_27
   widget=Background
   properties {
    visibility=SelfHitTestInvisible
    brush {
     bIsDynamicallyLoaded=false
     drawAs=Image
     imageType=FullColor
     mirroring=NoMirror
     resourceName=None
     resourceObject=/MatchmakingSettingsUI/MatchmakingSettings/MaterialInstances/MI_UI_MatchmakingSettings_Entry_Background
     tiling=NoTile
     imageSize=<32 32>
     margin=<ltrb 0 0 0 0>
     outlineSettings {
      bUseBrushTransparency=false
      roundingType=HalfHeightRadius
      width=0
      color {colorUseRule=UseColor_Specified, specifiedColor=<rgba 0 0 0 0>}
      cornerRadii=<0 0 0 0>
     }
     tintColor {colorUseRule=UseColor_Specified, specifiedColor=<rgba 1 1 1 1>}
     uVRegion {bIsValid=false, max=<0 0>, min=<0 0>}
    }
   }
   slot {class=OverlaySlot, properties {horizontalAlignment=HAlign_Fill, verticalAlignment=VAlign_Fill}}
  }
  ListView {
   widgetClassPath=&CommonListView
   bInherited=true
   bIsVariable=true
   parent=Overlay_27
   widget=ListView
   properties {
    bAllowDragging=false
    bEnableRightClickScrolling=false
    bReturnFocusToSelection=true
    bSelectItemOnNavigation=false
    clipping=Inherit
    consumeMouseWheel=Never
    entryWidgetClass=&/WBP_MatchmakingSettings_RadioButton.WBP_MatchmakingSettings_RadioButton_C
   }
   slot {class=OverlaySlot, properties {horizontalAlignment=HAlign_Fill, verticalAlignment=VAlign_Fill}}
  }
 ]
}
