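# Two-button matchmaking setting list entry that switches a discrete-enum setting, apparently the game mode, between options.
# A selection-required CommonButtonGroupBase holds the left and right entries, and SetButtonsData gives each a UIKitButtonViewModel labeled with an option name.
# Selection matches the button's label text to an option in the list item's FortMatchmakingSettingDiscreteEnumVM and sets that value for the owning player.
# Focus forwards to the selected button; compact UI adjusts minimum height.
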WidgetBlueprint /MatchmakingSettingsUI/MatchmakingSettings/SettingEntries/ModeSelection/WBP_MatchmakingSettings_ModeSelection
graphs [
 EventGraph {nodes [
   n1 = OnButtonSelected!() -> (OutputDelegate, then, AssociatedButton:Common Button Base Object Reference, ButtonIndex:int) {is=title K2Node_CustomEvent}
   > n2 = Cast To WBP_MatchmakingSettings_ModeSelection_Entry!(execute, Object=@n1.AssociatedButton) -> (then, CastFailed, AsWBP Matchmaking Settings Mode Selection Entry:WBP Matchmaking Settings Mode Selection Entry Object Reference) {is=title K2Node_DynamicCast}
   > n3 = Get!(execute:Exec, self:WBP Matchmaking Settings Mode Selection Entry Object Reference=@n2) -> (then:Exec, else:Exec, UIKitButtonViewModel:UI Kit Button View Model Object Reference)
   > n4 = Cast To FortMatchmakingSettingDiscreteEnumVM!(execute, Object=GetListItemObject!(UserObjectListEntry:User Object List Entry Interface) -> (ReturnValue:Object Reference)) -> (then, CastFailed, AsFort Matchmaking Setting Discrete Enum VM:Fort Matchmaking Setting Discrete Enum VM Object Reference) {is=title K2Node_DynamicCast}
   > n5 = FindOptionWithName!(execute, self, Name:"Text (by ref)"=Get Label!(self:UI Kit Button View Model Object Reference=@n3) -> (Label:Text)) -> (then, Option:String)
   > SetValue!(
    execute
    self:Fort Matchmaking Setting Discrete Enum VM Object Reference=@n4
    Value:String=@n5
    LocalPlayer:Local Player Object Reference=GetOwningLocalPlayer!(self:Widget Object Reference) -> (ReturnValue:Local Player Object Reference)
   ) -> (then)
   Event PreConstruct()
   > SetMobileStyle() {is=title K2Node_CallFunction}
   n6 = Event OnListItemObjectSet()
   > n7 = Cast To FortMatchmakingSettingDiscreteEnumVM(Object=@n6.ListItemObject)
   > Set FortMatchmakingSettingDiscreteEnumVM(FortMatchmakingSettingDiscreteEnumVM:Fort Matchmaking Setting Discrete Enum VM Object Reference=@n7) -> (Output_Get:Fort Matchmaking Setting Discrete Enum VM Object Reference)
   Event OnInitialized()
   > n8 = Construct Object from Class!(execute, Class=/Script/CommonUI.CommonButtonGroupBase, self, bSelectionRequired:bool=true) -> (then, ReturnValue:Common Button Group Base Object Reference)
   > n9 = Set ButtonGroup(ButtonGroup:Common Button Group Base Object Reference=@n8) -> (Output_Get:Common Button Group Base Object Reference)
   > Bind Event to On Selected Button Base Changed(self:Common Button Group Base Object Reference=@n9, Delegate=Create Event()) {is=title K2Node_AddDelegate}
   > AddWidgets!(
    execute
    self:Common Widget Group Base Object Reference=Get ButtonGroup() -> (ButtonGroup:Common Button Group Base Object Reference)
    Widgets:Array of Widget Object References=Make Array!(
     "[0]":Widget Object Reference=Get LeftEntry() -> (LeftEntry:WBP Matchmaking Settings Mode Selection Entry Object Reference)
     "[1]":Widget Object Reference=Get RightEntry() -> (RightEntry:WBP Matchmaking Settings Mode Selection Entry Object Reference)
    ) -> (Array:Array of Widget Object References)
   ) -> (then)
  ]}
 FindOptionWithName {nodes [
   n1 = FindOptionWithName!() -> (then, Name:"Text (by ref)") {is=title K2Node_FunctionEntry}
   > n2 = For Each Loop(Array:Array of Matchmaking Setting Option Structures=GetOptions!(self:Fort Matchmaking Setting Discrete Enum VM Object Reference=Get FortMatchmakingSettingDiscreteEnumVM() -> (FortMatchmakingSettingDiscreteEnumVM:Fort Matchmaking Setting Discrete Enum VM Object Reference)) -> (ReturnValue:Array of Matchmaking Setting Option Structures)) -> (Array Element:Matchmaking Setting Option Structure)
   Return Node!(execute=@n2.Completed, Option:String=Get Result() -> (Result:String))
   n3 = Break Matchmaking Setting Option!(MatchmakingSettingOption:"Matchmaking Setting Option Structure (by ref)"=@n2) -> (bIsEnabled:bool=~, Value:String, Name:Text) {is=title K2Node_BreakStruct}
   > Branch(Condition="Equal Exactly (Text)"!(A:"Text (by ref)"=@n3.Name, B:"Text (by ref)"=@n1) -> (ReturnValue:bool))
   > Set Result(Result:String=@n3.Value) -> (Output_Get:String)
  ]}
 OnFocusReceived {nodes [
   OnFocusReceived!() -> (then, MyGeometry:Geometry Structure, InFocusEvent:Focus Event Structure) {is=title K2Node_FunctionEntry}
   n1 = Get ButtonGroup()
   > Return Node!(
    execute
    ReturnValue:Event Reply Structure=SetUserFocus!(
     Reply:"Event Reply Structure (by ref)"=Handled()
     FocusWidget:Widget Object Reference=GetSelectedButtonBase!(self:Common Button Group Base Object Reference=@n1) -> (ReturnValue:Common Button Base Object Reference)
     bInAllUsers:bool=false
    ) -> (ReturnValue:Event Reply Structure)
   )
  ]}
 SetButtonsData {nodes [
   n1 = SetButtonsData!() -> (then, Options:Array of Matchmaking Setting Option Structures) {is=title K2Node_FunctionEntry}
   > n2 = For Each Loop(Array=@n1)
   > n3 = Cast To WBP_MatchmakingSettings_ModeSelection_Entry(Object=GetButtonBaseAtIndex!(self:Common Button Group Base Object Reference=Get ButtonGroup(), Index:int=@n2.Array Index) -> (ReturnValue:Common Button Base Object Reference))
   > n4 = Construct Object from Class!(execute, Class=/Script/FortniteUI.UIKitButtonViewModel, self) -> (then, ReturnValue:UI Kit Button View Model Object Reference)
   > Set Label!(
    execute
    Label:Text=Break Matchmaking Setting Option(MatchmakingSettingOption=@n2) -> (bIsEnabled=~).Name
    self:UI Kit Button View Model Object Reference=@n4
   ) -> (then, Output_Get:Text)
   > Set UIKitButtonViewModel!(
    execute
    UIKitButtonViewModel:UI Kit Button View Model Object Reference=@n4
    self:WBP Matchmaking Settings Mode Selection Entry Object Reference=@n3
   ) -> (then, Output_Get:UI Kit Button View Model Object Reference)
  ]}
 SetMobileStyle {nodes [
   SetMobileStyle() {is=title K2Node_FunctionEntry}
   > Set Size!(
    execute
    Size:Vector 2D Structure=Select(Option 0:Vector 2D Structure=(X=8,Y=64), Option 1:Vector 2D Structure=(X=8,Y=80), Index=IsUsingCompactUI()) -> (ReturnValue:Vector 2D Structure)
    self:Spacer Object Reference=Get Min_Height() -> (Min_Height:Spacer Object Reference)
   ) -> (then, Output_Get:Vector 2D Structure)
  ]}
]
variables [ButtonGroup {type=CommonButtonGroupBase, default=null}]
widgets {
 info {inheritedWidgetCount=0, parentClass=&UserWidget, rootWidgetClass=&Overlay, widgetCount=6}
 widgets [
  Overlay_0 {widgetClassPath=&Overlay, widget=Overlay_0, properties {bOverride_Cursor=true}}
  Min_Height {
   widgetClassPath=&Spacer
   bIsVariable=true
   parent=Overlay_0
   widget=Min_Height
   properties {size=<1 64>}
   slot {class=OverlaySlot, properties {horizontalAlignment=HAlign_Fill, verticalAlignment=VAlign_Fill}}
  }
  Background {
   widgetClassPath=&Image
   bIsVariable=true
   parent=Overlay_0
   widget=Background
   properties {
    visibility=SelfHitTestInvisible
    brush {
     bIsDynamicallyLoaded=false
     drawAs=Image
     imageType=FullColor
     mirroring=NoMirror
     resourceName=None
     resourceObject=/MatchmakingSettingsUI/MatchmakingSettings/MaterialInstances/MI_UI_MatchmakingSettings_Entry_Background
     tiling=NoTile
     imageSize=<32 32>
     margin=<ltrb 0 0 0 0>
     outlineSettings {
      bUseBrushTransparency=false
      roundingType=HalfHeightRadius
      width=0
      color {colorUseRule=UseColor_Specified, specifiedColor=<rgba 0 0 0 0>}
      cornerRadii=<0 0 0 0>
     }
     tintColor {colorUseRule=UseColor_Specified, specifiedColor=<rgba 1 1 1 1>}
     uVRegion {bIsValid=false, max=<0 0>, min=<0 0>}
    }
   }
   slot {class=OverlaySlot, properties {horizontalAlignment=HAlign_Fill, verticalAlignment=VAlign_Fill}}
  }
  HorizontalBox_0 {
   widgetClassPath=&HorizontalBox
   parent=Overlay_0
   widget=HorizontalBox_0
   slot {class=OverlaySlot, properties {horizontalAlignment=HAlign_Fill, verticalAlignment=VAlign_Fill, padding=<ltrb 4 4 4 4>}}
  }
  LeftEntry {
   widgetClassPath=&/WBP_MatchmakingSettings_ModeSelection_Entry
   bIsVariable=true
   parent=HorizontalBox_0
   widget=LeftEntry
   properties {
    bSelectable=true
    holdData=&CommonUI_HoldData_Base_C
    navigation=&self:WidgetTree.LeftEntry.WidgetNavigation_0
   }
   slot {class=HorizontalBoxSlot, properties {size {sizeRule=Fill, value=1}}}
  }
  RightEntry {
   widgetClassPath=&WBP_MatchmakingSettings_ModeSelection_Entry
   bIsVariable=true
   parent=HorizontalBox_0
   widget=RightEntry
   properties {bSelectable=true, holdData=&CommonUI_HoldData_Base_C}
   slot {class=HorizontalBoxSlot, properties {padding=<ltrb 12 0 0 0>, size {sizeRule=Fill, value=1}}}
  }
 ]
}
