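# Matchmaking settings list entry widget for ranked play, showing the player's rank icon and name beside an on/off toggle.
# On list binding it stores the discrete enum setting view model and caches its first two option values as false and true.
# Clicks flip between those values, and HandleSettingValueChanged syncs the toggle.
# SetCurrentPlayerRank resolves rank text and icon from Habanero display data with overrides, keyed by ranked view model data.
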WidgetBlueprint /MatchmakingSettingsUI/MatchmakingSettings/SettingEntries/Ranked/WBP_MatchmakingSettings_Ranked
graphs [
 EventGraph {nodes [
   n1 = Event OnListItemObjectSet()
   > n2 = Cast To FortMatchmakingSettingDiscreteEnumVM!(execute, Object=@n1.ListItemObject) -> (then, CastFailed, AsFort Matchmaking Setting Discrete Enum VM:Fort Matchmaking Setting Discrete Enum VM Object Reference) {is=title K2Node_DynamicCast}
   > Set FortSettingVM(FortSettingVM:Fort Matchmaking Setting Discrete Enum VM Object Reference=@n2) -> (Output_Get:Fort Matchmaking Setting Discrete Enum VM Object Reference)
   n3 = GetOptions!(self:Fort Matchmaking Setting Discrete Enum VM Object Reference=@n2) -> (ReturnValue:Array of Matchmaking Setting Option Structures)
   > Set False Value(False Value:String=Break Matchmaking Setting Option!(MatchmakingSettingOption:"Matchmaking Setting Option Structure (by ref)"="Get (a ref)"(Array:Array of Matchmaking Setting Option Structures=@n3, Dimension 1=0) -> (Output:"Matchmaking Setting Option Structure (by ref)")) -> (bIsEnabled:bool=~, Value:String, Name:Text) {is=title K2Node_BreakStruct}.Value) -> (Output_Get:String)
   > Set True Value(True Value:String=Break Matchmaking Setting Option(MatchmakingSettingOption="Get (a ref)"(Array=@n3, Dimension 1=1)) -> (bIsEnabled=~).Value) -> (Output_Get:String)
   Event On Clicked()
   n4 = Get FortSettingVM() -> (FortSettingVM:Fort Matchmaking Setting Discrete Enum VM Object Reference)
   n5 = GetValue!(self:Fort Matchmaking Setting Discrete Enum VM Object Reference=@n4) -> (ReturnValue:String)
   n6 = Get False Value() -> (False Value:String)
   > SetValue!(
    execute
    self:Fort Matchmaking Setting Discrete Enum VM Object Reference=@n4
    Value:String=SelectString(A=Get True Value() -> (True Value:String), B=Get False Value(), bPickA="Equal Exactly (String)"(A=@n5, B=@n6))
    LocalPlayer:Local Player Object Reference=GetOwningLocalPlayer!(self:Widget Object Reference) -> (ReturnValue:Local Player Object Reference)
   ) -> (then)
   Event OnInitialized()
   > InitializeRankedViewModal(self:Ranked View Model Object Reference=Get FortRankedViewModel() -> (FortRankedViewModel:Ranked View Model Object Reference))
  ]}
 HandleSettingValueChanged {nodes [
   n1 = HandleSettingValueChanged!() -> (then, Value:String) {is=title K2Node_FunctionEntry}
   > SetToggle!(
    execute
    self:WBP UIKit Toggle Object Reference=Get Toggle() -> (Toggle:WBP UIKit Toggle Object Reference)
    On:bool="Equal Exactly (String)"(A=@n1, B=Get True Value())
   ) -> (then)
  ]}
 SetCurrentPlayerRank {nodes [
   n1 = SetCurrentPlayerRank!() -> (then, CurrentRank:int) {is=title K2Node_FunctionEntry}
   n2 = Get FortRankedViewModel()
   n3 = Get RankType!(self:Ranked View Model Object Reference=@n2) -> (RankType:String)
   n4 = GetHabaneroDisplayDataOverride!(
    InDisplayData:Fort Habanero Display Data Object Reference=Get RankDisplayData() -> (RankDisplayData:Fort Habanero Display Data Object Reference)
    InOverrideData:Fort Compete UIHabanero Tier Override List Object Reference=Get Rank Info Overrides() -> (Rank Info Overrides:Fort Compete UIHabanero Tier Override List Object Reference)
    InKey:"Fort Compete UIHabanero Tier Override Key Structure (by ref)"=Make Fort Compete UIHabanero Tier Override Key!(
     RankedTrack:Name=String To Name(InString=@n3)
     Season:String
     Rank:int=@n1
     MatchmakingRating:int=Get MatchmakingRating!(self:Ranked View Model Object Reference=@n2) -> (MatchmakingRating:int=0)
     CurrentPosition:int=Get CurrentPlayerPosition!(self:Ranked View Model Object Reference=@n2) -> (CurrentPlayerPosition:int=0)
    ) -> (FortCompeteUIHabaneroTierOverrideKey:Fort Compete UIHabanero Tier Override Key Structure) {is=title K2Node_MakeStruct}
   ) -> (OutData:Fort Habanero Tier Override Structure, ReturnValue:bool)
   > Branch(Condition=@n4.ReturnValue)
   n5 = Break Fort Habanero Tier!(FortHabaneroTier:"Fort Habanero Tier Structure (by ref)"=@n4) -> (ForegroundIconSmall:Material Instance Soft Object Reference, BackgroundIconSmall:Material Instance Soft Object Reference, Name:Text) {is=title K2Node_BreakStruct}
   > "SetText (Text)"(self=Get Txt_Rank() -> (Txt_Rank:Common Text Object Reference), InText=@n5.Name)
   > UpdateDetails!(
    execute
    self:WBP Habanero Icon Object Reference=Get WBP_Habanero_Icon() -> (WBP_Habanero_Icon:WBP Habanero Icon Object Reference)
    ForegroundIcon:Material Instance Soft Object Reference=@n5
    BackgroundIcon:Material Instance Soft Object Reference=@n5.BackgroundIconSmall
   ) -> (then)
  ]}
]
variables [
 False Value {type=string, default=""}
 Rank Info Overrides {type=FortCompeteUIHabaneroTierOverrideList, default=/Game/UI/Habanero/DA_RankDisplayDataOverride}
 RankDisplayData {type=FortHabaneroDisplayData, default=/Game/UI/Habanero/HabaneroDisplayData}
 True Value {type=string, default=""}
]
widgets {
 info {
  inheritedWidgetCount=0
  namedSlotCount=1
  parentClass=/MatchmakingSettingsUI/MatchmakingSettings/SettingEntries/WBP_MatchmakingSettings_EntryBase.WBP_MatchmakingSettings_EntryBase_C
  rootWidgetClass=null
  widgetCount=4
 }
 widgets [
  HorizontalBox_0 {widgetClassPath=&HorizontalBox, widget=HorizontalBox_0}
  Txt_Rank {
   widgetClassPath=&CommonTextBlock
   bIsVariable=true
   parent=HorizontalBox_0
   widget=Txt_Rank
   properties {
    text=$Rank
    colorAndOpacity {colorUseRule=UseColor_Specified, specifiedColor=<rgba 0.5776 0.5776 0.5776 1>}
    font {
     bForceMonospaced=false
     bMaterialIsStencil=false
     fontMaterial=null
     fontObject=&NotoSans
     letterSpacing=0
     monospacedWidth=1
     size=24
     skewAmount=0
     typefaceFontName=NotoSans-Regular
     outlineSettings {
      bApplyOutlineToDropShadows=false
      bMiteredCorners=false
      bSeparateFillAlpha=false
      outlineMaterial=null
      outlineSize=0
      outlineColor=<rgba 0 0 0 1>
     }
    }
    shadowOffset=<0 0>
   }
   slot {class=HorizontalBoxSlot, properties {horizontalAlignment=HAlign_Left, verticalAlignment=VAlign_Center, padding=<ltrb 0 0 8 0>}}
  }
  WBP_Habanero_Icon {
   widgetClassPath=/Game/UI/Frontend/Lobby/Habanero/WBP_Habanero_Icon
   bIsVariable=true
   parent=HorizontalBox_0
   widget=WBP_Habanero_Icon
   properties {iconMinSize=<48 48>}
   slot {class=HorizontalBoxSlot, properties {padding=<ltrb 0 0 16 0>}}
  }
  Toggle {
   widgetClassPath=/MatchmakingSettingsUI/MatchmakingSettings/SettingEntries/Toggle/WBP_UIKit_Toggle
   bIsVariable=true
   parent=HorizontalBox_0
   widget=Toggle
   slot {class=HorizontalBoxSlot, properties {verticalAlignment=VAlign_Center}}
  }
 ]
}
