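# Matchmaking settings list entry widget that shows a numeric setting as a UIKit slider with a value label.
# When set as a list item, it casts the item to FortMatchmakingSettingScalarVM and stores it.
# Slider changes, bound on Construct and cleared on Destruct, write the value to that view model.
# HandleOptionValueChanged syncs a given value to the slider, bar material and text, with up to three decimals.
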WidgetBlueprint /MatchmakingSettingsUI/MatchmakingSettings/SettingEntries/Slider/WBP_MatchmakingSettings_Slider_Entry
graphs [
 EventGraph {nodes [
   n1 = OnValueChanged!() -> (OutputDelegate, then, Value:float=~) {is=title K2Node_CustomEvent}
   n2 = Get InteractionLayer_Slider!(self:WBP UIKit Slider Base Object Reference=Get Slider() -> (Slider:WBP UIKit Slider Object Reference)) -> (InteractionLayer_Slider:Slider Object Reference)
   > SetValue!(execute, self:Slider Object Reference=@n2, InValue:float=@n1.Value) -> (then)
   > SetValue!(
    execute
    self:Fort Setting Scalar VM Object Reference=Get FortSettingVM() -> (FortSettingVM:Fort Matchmaking Setting Scalar VM Object Reference)
    Value:double=@n1.Value
   ) -> (then)
   n3 = Event OnListItemObjectSet()
   > n4 = Cast To FortMatchmakingSettingScalarVM!(execute, Object=@n3.ListItemObject) -> (then, CastFailed, AsFort Matchmaking Setting Scalar VM:Fort Matchmaking Setting Scalar VM Object Reference) {is=title K2Node_DynamicCast}
   > Set FortSettingVM(FortSettingVM:Fort Matchmaking Setting Scalar VM Object Reference=@n4) -> (Output_Get:Fort Matchmaking Setting Scalar VM Object Reference)
   Event Construct()
   > Bind Event to On Value Changed(self:Slider Object Reference=@n2, Delegate=@n1) {is=title K2Node_AddDelegate}
   Event Destruct()
   > Unbind all Events from On Value Changed!(execute:Exec, self:Slider Object Reference=@n2) -> (then:Exec) {is=title K2Node_ClearDelegate}
  ]}
 HandleOptionValueChanged {nodes [
   n1 = HandleOptionValueChanged!() -> (then, Value:double=~) {is=title K2Node_FunctionEntry}
   n2 = Get Slider()
   > SetValue!(execute, self:Slider Object Reference=Get InteractionLayer_Slider(self=@n2), InValue:float=@n1) -> (then)
   > UpdateBarMaterial!(execute, self:WBP UIKit Slider Object Reference=@n2, Value:double=@n1) -> (then)
   > "SetText (Text)"(
    self=Get Txt_SliderCurrentValue() -> (Txt_SliderCurrentValue:Common Text Object Reference)
    InText="To Text (Float)"(Value=@n1, MinimumIntegralDigits=1, MaximumIntegralDigits=324, MinimumFractionalDigits=0, MaximumFractionalDigits=3)
   )
  ]}
]
variables []
widgets {
 info {
  inheritedWidgetCount=0
  namedSlotCount=1
  parentClass=/MatchmakingSettingsUI/MatchmakingSettings/SettingEntries/WBP_MatchmakingSettings_EntryBase.WBP_MatchmakingSettings_EntryBase_C
  rootWidgetClass=null
  widgetCount=3
 }
 widgets [
  HorizontalBox_0 {widgetClassPath=&HorizontalBox, widget=HorizontalBox_0}
  Txt_SliderCurrentValue {
   widgetClassPath=&CommonTextBlock
   bIsVariable=true
   parent=HorizontalBox_0
   widget=Txt_SliderCurrentValue
   properties {
    style=&TextStyle-Base_C
    text=$0
    font {
     bForceMonospaced=false
     bMaterialIsStencil=false
     fontMaterial=null
     fontObject=&NotoSans
     letterSpacing=0
     monospacedWidth=1
     size=24
     skewAmount=0
     typefaceFontName=NotoSans-Regular
     outlineSettings {
      bApplyOutlineToDropShadows=false
      bMiteredCorners=false
      bSeparateFillAlpha=false
      outlineMaterial=null
      outlineSize=0
      outlineColor=<rgba 0 0 0 1>
     }
    }
    shadowOffset=<0 0>
   }
   slot {class=HorizontalBoxSlot, properties {horizontalAlignment=HAlign_Right, verticalAlignment=VAlign_Center, padding=<ltrb 0 0 8 0>}}
  }
  Slider {
   widgetClassPath=&/WBP_UIKit_Slider
   bIsVariable=true
   parent=HorizontalBox_0
   widget=Slider
   slot {class=HorizontalBoxSlot, properties {verticalAlignment=VAlign_Center}}
  }
 ]
}
