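# Toggle entry widget for a two-option matchmaking setting in a list, derived from the matchmaking settings entry base.
# When its list item is set, it casts to a discrete-enum setting view model and stores the first two options' values as false and true.
# Clicking writes the opposite value back through the view model for the owning local player.
# A value-change handler drives the UIKit toggle by comparing against the true value.
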WidgetBlueprint /MatchmakingSettingsUI/MatchmakingSettings/SettingEntries/Toggle/WBP_MatchmakingSettings_Toggle
graphs [
 EventGraph {nodes [
   n1 = Event OnListItemObjectSet()
   > n2 = Cast To FortMatchmakingSettingDiscreteEnumVM!(execute, Object=@n1.ListItemObject) -> (then, CastFailed, AsFort Matchmaking Setting Discrete Enum VM:Fort Matchmaking Setting Discrete Enum VM Object Reference) {is=title K2Node_DynamicCast}
   n3 = GetOptions!(self:Fort Matchmaking Setting Discrete Enum VM Object Reference=@n2) -> (ReturnValue:Array of Matchmaking Setting Option Structures)
   > Set False Value(False Value:String=Break Matchmaking Setting Option!(MatchmakingSettingOption:"Matchmaking Setting Option Structure (by ref)"="Get (a ref)"(Array:Array of Matchmaking Setting Option Structures=@n3, Dimension 1=0) -> (Output:"Matchmaking Setting Option Structure (by ref)")) -> (bIsEnabled:bool=~, Value:String, Name:Text) {is=title K2Node_BreakStruct}.Value) -> (Output_Get:String)
   > Set True Value(True Value:String=Break Matchmaking Setting Option(MatchmakingSettingOption="Get (a ref)"(Array=@n3, Dimension 1=1)) -> (bIsEnabled=~).Value) -> (Output_Get:String)
   > Set FortSettingVM(FortSettingVM:Fort Matchmaking Setting Discrete Enum VM Object Reference=@n2) -> (Output_Get:Fort Matchmaking Setting Discrete Enum VM Object Reference)
   Event On Clicked()
   n4 = Get FortSettingVM() -> (FortSettingVM:Fort Matchmaking Setting Discrete Enum VM Object Reference)
   n5 = GetValue!(self:Fort Matchmaking Setting Discrete Enum VM Object Reference=@n4) -> (ReturnValue:String)
   n6 = Get False Value() -> (False Value:String)
   > SetValue!(
    execute
    self:Fort Matchmaking Setting Discrete Enum VM Object Reference=@n4
    Value:String=SelectString(A=Get True Value() -> (True Value:String), B=Get False Value(), bPickA="Equal Exactly (String)"(A=@n5, B=@n6))
    LocalPlayer:Local Player Object Reference=GetOwningLocalPlayer!(self:Widget Object Reference) -> (ReturnValue:Local Player Object Reference)
   ) -> (then)
  ]}
 HandleSettingValueChanged {nodes [
   n1 = HandleSettingValueChanged!() -> (then, Value:String) {is=title K2Node_FunctionEntry}
   > SetToggle!(
    execute
    self:WBP UIKit Toggle Object Reference=Get Toggle() -> (Toggle:WBP UIKit Toggle Object Reference)
    On:bool="Equal Exactly (String)"(A=@n1, B=Get True Value())
   ) -> (then)
  ]}
]
variables [
 False Value {type=string, default=""}
 True Value {type=string, default=""}
]
widgets {
 info {
  inheritedWidgetCount=0
  namedSlotCount=1
  parentClass=/MatchmakingSettingsUI/MatchmakingSettings/SettingEntries/WBP_MatchmakingSettings_EntryBase.WBP_MatchmakingSettings_EntryBase_C
  rootWidgetClass=null
  widgetCount=2
 }
 widgets [
  Overlay_1 {widgetClassPath=&Overlay, widget=Overlay_1}
  Toggle {
   widgetClassPath=&/WBP_UIKit_Toggle
   bIsVariable=true
   parent=Overlay_1
   widget=Toggle
   slot {class=OverlaySlot, properties {verticalAlignment=VAlign_Center}}
  }
 ]
}
