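# Polymorphic list view widget for matchmaking settings, built on FortPolymorphicListView, that shows a different entry widget per setting type.
# GetDesiredEntryClassForItem casts each item to a discrete or scalar setting view model and looks up its variant name in a name-to-class map.
# The map covers toggle, dropdown, enum selector, mode selection, ranked and slider entries.
# Setting items are selectable only when enabled; other items always are.
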WidgetBlueprint /MatchmakingSettingsUI/MatchmakingSettings/WBP_MatchmakingSettings_PolyList
graphs [
 EventGraph {nodes [
   Event PreConstruct()
   Event Construct()
   Event Tick()
  ]}
 GetDesiredEntryClassForItem {nodes [
   n1 = GetDesiredEntryClassForItem!() -> (then, Item:Object Reference) {is=title K2Node_FunctionEntry}
   > n2 = Cast To FortMatchmakingSettingDiscreteVM!(execute, Object=@n1) -> (then, CastFailed, AsFort Matchmaking Setting Discrete VM:Fort Matchmaking Setting Discrete VM Object Reference) {is=title K2Node_DynamicCast}
   > n3 = Set VariantKey(VariantKey:Name=GetVariant!(self:Fort Matchmaking Setting Discrete VM Object Reference=@n2) -> (ReturnValue:Name=None)) -> (Output_Get:Name=None)
   n4 = Find(
    TargetMap:Map of Names to User Widget Class References=Get VMTypeToWidget() -> (VMTypeToWidget:Map of Names to User Widget Class References)
    Key:"Name (by ref)"=Get VariantKey() -> (VariantKey:Name=None)
   ) -> (Value:User Widget Class Reference)
   n5 = Branch(execute=[@n7.then, @n3.then], Condition=@n4.ReturnValue)
   > Return Node!(execute, ReturnValue:User Widget Class Reference=@n4)
   Return Node(execute=[@n5.else, @n6.CastFailed])
   n6 = Cast To FortMatchmakingSettingScalarVM!(execute=@n2.CastFailed, Object=@n1) -> (then, CastFailed, AsFort Matchmaking Setting Scalar VM:Fort Matchmaking Setting Scalar VM Object Reference) {is=title K2Node_DynamicCast}
   > n7 = Set VariantKey(VariantKey=GetVariant(self:Fort Matchmaking Setting Scalar VM Object Reference=@n6) -> (ReturnValue=None)) -> (Output_Get=None)
  ]}
 GetItemSelectableOrNavigable {nodes [
   n1 = GetItemSelectableOrNavigable!() -> (then, Item:Object Reference) {is=title K2Node_FunctionEntry}
   > n2 = Cast To FortSettingVM!(execute, Object=@n1) -> (then, CastFailed, AsFort Setting VM:Fort Setting VM Object Reference) {is=title K2Node_DynamicCast}
   > Return Node(ReturnValue:bool=GetEnabled!(self:Fort Setting VM Object Reference=@n2) -> (ReturnValue:bool))
   Return Node(execute=@n2.CastFailed, ReturnValue=true)
  ]}
]
variables [VMTypeToWidget {
  type="map<Class>"
  default {
   checkbox=/MatchmakingSettingsUI/MatchmakingSettings/SettingEntries/Toggle/WBP_MatchmakingSettings_Toggle.WBP_MatchmakingSettings_Toggle_C
   dropdown=/MatchmakingSettingsUI/MatchmakingSettings/SettingEntries/Dropdown/WBP_MatchmakingSettings_Dropdown.WBP_MatchmakingSettings_Dropdown_C
   enumSelector=/MatchmakingSettingsUI/MatchmakingSettings/SettingEntries/EnumSelector/WBP_MatchmakingSettings_EnumSelector.WBP_MatchmakingSettings_EnumSelector_C
   modeSelection=/MatchmakingSettingsUI/MatchmakingSettings/SettingEntries/ModeSelection/WBP_MatchmakingSettings_ModeSelection.WBP_MatchmakingSettings_ModeSelection_C
   ranked=/MatchmakingSettingsUI/MatchmakingSettings/SettingEntries/Ranked/WBP_MatchmakingSettings_Ranked.WBP_MatchmakingSettings_Ranked_C
   scalar=/MatchmakingSettingsUI/MatchmakingSettings/SettingEntries/Slider/WBP_MatchmakingSettings_Slider_Entry.WBP_MatchmakingSettings_Slider_Entry_C
   toggle=&WBP_MatchmakingSettings_Toggle_C
  }
 }]
widgets {
 info {
  inheritedWidgetCount=1
  parentClass=/Script/FortniteUI.FortPolymorphicListView
  rootWidgetClass=&CommonListView
  widgetCount=1
 }
 widgets [ListView {
   widgetClassPath=&CommonListView
   bInherited=true
   bIsVariable=true
   widget=ListView
   properties {
    bReturnFocusToSelection=true
    clipping=Inherit
    entryWidgetClass=&WBP_MatchmakingSettings_Dropdown_C
    verticalEntrySpacing=12
   }
  }]
}
