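# Matchmaking Play button, derived from the generic UIKit button, that toggles ready-up through the SelectedIslandViewModel from MatchmakingUISubsystem.
# Clicking calls ReadyUp or Unready, then PlayReadyAudio, which despite its name only swaps Play style-sheet tags.
# Helper functions derive label and lock state from error UI, matchmaking status and, for the label, the owning player's ready-up status.
# PreConstruct clips the button and adjusts its background and outline material corners.
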WidgetBlueprint /MatchmakingSettingsUI/MatchmakingSettings/WBP_UI_Button_Matchmaking
graphs [
 EventGraph {nodes [
   n1 = Event PreConstruct()
   > Parent: PreConstruct(IsDesignTime=@n1.IsDesignTime)
   > Set Clipping!(execute, Clipping:EWidgetClipping Enum=ClipToBounds, self:Widget Object Reference=Get Main() -> (Main:Grid Panel Object Reference)) -> (then, Output_Get:EWidgetClipping Enum=Inherit)
   > n2 = Cast To WBP_UIKit_Block_BackgroundWithOutline!(execute, Object=Get Block_Background() -> (Block_Background:WBP UIKit Block Background Base Object Reference)) -> (then, CastFailed, AsWBP UIKit Block Background with Outline:WBP UIKit Block Background With Outline Object Reference) {is=title K2Node_DynamicCast}
   n3 = Get BackgroundDynamicMaterial!(self:WBP UIKit Block Background Base Object Reference=@n2) -> (BackgroundDynamicMaterial:Material Instance Dynamic Object Reference)
   n4 = Get OutlineDynamicMaterial!(self:WBP UIKit Block Background Base Object Reference=@n2) -> (OutlineDynamicMaterial:Material Instance Dynamic Object Reference)
   > SetScalarParameterValue(self=[@n3, @n4], ParameterName=Radius in Pixels, Value=0)
   > SetScalarParameterValue(self=@n3, ParameterName=Corner Radius, Value=0.3)
   > SetVectorParameterValue!(execute, self=@n4, ParameterName=Corner Radius, Value_R:float=0.325, Value_G:float=0.325, Value_B:float=0.325, Value_A:float=0.325) -> (then)
   Event Construct()
   > Set SelectedIslandViewModel(SelectedIslandViewModel:Selected Island View Model Object Reference=GetSelectedIslandViewModel!(self:Matchmaking UISubsystem Object Reference=Get MatchmakingUISubsystem() -> (ReturnValue:Matchmaking UISubsystem Object Reference) {is=title K2Node_GetSubsystem}) -> (ReturnValue:Selected Island View Model Object Reference)) -> (Output_Get:Selected Island View Model Object Reference)
   Event On Clicked()
   n5 = Get SelectedIslandViewModel() -> (SelectedIslandViewModel:Selected Island View Model Object Reference)
   > n6 = Branch(Condition="Equal (Enum)"(
     A:EFortUIMatchmakingStatus Enum=Get MatchmakingStatus!(self:Selected Island View Model Object Reference=@n5) -> (MatchmakingStatus:EFortUIMatchmakingStatus Enum=InvalidIsland)
     B:EFortUIMatchmakingStatus Enum=Unready
    ))
   n7 = Get SelectedIslandViewModel()
   > ReadyUp(self:Selected Island View Model Object Reference=@n7)
   > PlayReadyAudio!(execute, self, bReady:bool=false) -> (then)
   Unready(execute=@n6.else, self:Selected Island View Model Object Reference=@n7)
   > PlayReadyAudio(bReady=true)
   Event Destruct()
   > Parent: Destruct()
  ]}
 GetPlayButtonText {nodes [
   n1 = GetPlayButtonText!() -> (
    then
    MatchmakingStatus:EFortUIMatchmakingStatus Enum=Unready
    Error UI:Ready Up Error UI Object Reference
    ReadyUp Status:Map of Unique Net Id Structures to EFortUIReadyUpStatus Enums
   ) {is=title K2Node_FunctionEntry}
   > n2 = Branch(Condition=IsValid(Object=@n1.Error UI))
   n3 = Play Button Text!(self:Ready Up Error UI Object Reference=Get Error UI() -> (Error UI:Ready Up Error UI Object Reference)) -> (ReturnValue:Text)
   > n4 = Branch(Condition=NOT Boolean(A=TextIsEmpty(InText=@n3)))
   > Return Node!(execute, PlayButtonText:Text=Play Button Text(self=Get Error UI()))
   n5 = Get MatchmakingStatus!() -> (MatchmakingStatus=InvalidIsland)
   n6 = Branch(
    execute=[@n4.else, @n2.else]
    Condition=AND Boolean(
     A="Not Equal (Enum)"!(A:EFortUIMatchmakingStatus Enum=@n5, B:EFortUIMatchmakingStatus Enum=Unready) -> (ReturnValue:bool)
     B="Not Equal (Enum)"(A=@n5, B=InvalidIsland)
    )
   )
   > Return Node(PlayButtonText="LOCTEXT(\"Cancel\")")
   n7 = GetOwningPlayer()
   n8 = Switch on EFortUIReadyUpStatus!(
    execute=@n6.else
    Selection:EFortUIReadyUpStatus Enum=Find(
     TargetMap:Map of Unique Net Id Structures to EFortUIReadyUpStatus Enums=Get ReadyUp Status() -> (ReadyUp Status:Map of Unique Net Id Structures to EFortUIReadyUpStatus Enums)
     Key:"Unique Net Id Structure (by ref)"=GetNetIDFromContextObject!(ContextObject:Object Reference=@n7) -> (ReturnValue:Unique Net Id Structure)
    ) -> (Value:EFortUIReadyUpStatus Enum=Invalid)
   ) -> (Invalid:Exec, NotPlaying:Exec, Play:Exec, Ready:Exec, Join:Exec, Rejoin:Exec, Watch:Exec) {is=title K2Node_SwitchEnum}
   Return Node(execute=@n8.Join, PlayButtonText="LOCTEXT(\"Join\")")
   Return Node(execute=@n8.Play, PlayButtonText="LOCTEXT(\"Play\")")
   Return Node(execute=[@n8.Ready, @n8.NotPlaying], PlayButtonText="LOCTEXT(\"Ready\")")
   Return Node(execute=@n8.Rejoin, PlayButtonText="LOCTEXT(\"Rejoin\")")
   Return Node(execute=@n8.Watch, PlayButtonText="LOCTEXT(\"Watch\")")
  ]}
 PlayReadyAudio {nodes [
   n1 = PlayReadyAudio!() -> (then, bReady:bool=~) {is=title K2Node_FunctionEntry}
   > n2 = Branch(Condition=@n1)
   > UpdateWidgetStyleTags!(
    execute
    InWidget:Widget Object Reference=Self-Reference()
    InAddedTags:Array of Gameplay Tag Structures=Get PlayStyleSheetTags() -> (PlayStyleSheetTags:Array of Gameplay Tag Structures)
    InRemovedTags:Array of Gameplay Tag Structures=Get RemovedStyleSheetTags() -> (RemovedStyleSheetTags:Array of Gameplay Tag Structures)
   ) -> (then, ReturnValue:bool)
   UpdateWidgetStyleTags(execute=@n2.else, InWidget=Self-Reference(), InAddedTags=Get RemovedStyleSheetTags(), InRemovedTags=Get PlayStyleSheetTags())
  ]}
 Sequencer Events {links [], nodes []}
 Set Play Button Enabled {nodes [
   n1 = Set Play Button Enabled!() -> (then, ErrorUI:Ready Up Error UI Object Reference, MatchmakingStatus:EFortUIMatchmakingStatus Enum=Unready) {is=title K2Node_FunctionEntry}
   > n2 = Branch(Condition=IsValid(Object=@n1))
   n3 = Get ErrorUI() -> (ErrorUI:Ready Up Error UI Object Reference)
   > Return Node!(execute, IsPlayLocked:bool=NOT Boolean(A=HasRedirect!(self:Ready Up Error UI Object Reference=@n3) -> (ReturnValue:bool)))
   n4 = Get MatchmakingStatus() -> (MatchmakingStatus=InvalidIsland)
   n5 = "Not Equal (Enum)"(A=@n4, B=InvalidIsland)
   n6 = "Not Equal (Enum)"(A=MatchFound, B=@n4)
   Return Node(execute=@n2.else, IsPlayLocked=NOT Boolean(A=AND Boolean(A=@n6, B=@n5)))
  ]}
]
variables [
 PlayButtonText {type=string, default=""}
 PlayStyleSheetTags {type="array<GameplayTag>", default [{tagName=StyleSheet.Style.CTA}, {tagName=StyleSheet.Style.Loud}, {tagName=StyleSheet.Style.Play}]}
 RemovedStyleSheetTags {type="array<GameplayTag>", default []}
]
widgets {
 info {
  inheritedWidgetCount=0
  parentClass=/FortUILibrary/UIKit/Buttons/WBP_UIKit_Button_Generic.WBP_UIKit_Button_Generic_C
  rootWidgetClass=null
  widgetCount=0
 }
 widgets []
}
