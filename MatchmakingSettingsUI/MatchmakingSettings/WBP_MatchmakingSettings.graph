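# Matchmaking settings screen for the selected island, with a settings list, map rotation entry, description panel and Play/Close buttons that both deactivate it.
# On activation it takes the island view model from MatchmakingUISubsystem, shares it with map rotation and selects the first setting.
# Hovered settings fill the panel text, while focus targets the selected row.
# HandleIslandChanged loads the fullscreen image, player-count badge and asynchronous logo widget, with text fallback.
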WidgetBlueprint /MatchmakingSettingsUI/MatchmakingSettings/WBP_MatchmakingSettings
animations [Anim_Intro {
  binding_count=5
  length_seconds=1.0667
  object=&self:Anim_Intro
  sequence {
   bindings [
    SB_Content {
     id=#1
     possessed_class=SizeBox
     tracks [
      {
       type=MovieSceneFloatTrack
       display_name=Render Opacity
       sections [{type=MovieSceneFloatSection, end_frame=65, start_frame=1, channels [None {type=MovieSceneScriptingFloatChannel, keys [[1,0],[6,0],[65,1]]}]}]
      }
      {
       type=MovieScene2DTransformTrack
       sections [{
         type=MovieScene2DTransformSection
         end_frame=65
         start_frame=1
         channels [Translation.Y {type=MovieSceneScriptingFloatChannel, keys [[1,100],[6,100],[65,0]]}]
        }]
      }
     ]
    }
    O_ForTopPadding {
     id=#2
     possessed_class=Overlay
     tracks [
      {
       type=MovieSceneFloatTrack
       display_name=Render Opacity
       sections [{type=MovieSceneFloatSection, end_frame=60, start_frame=1, channels [None {type=MovieSceneScriptingFloatChannel, keys [[1,0],[60,1]]}]}]
      }
      {
       type=MovieScene2DTransformTrack
       sections [{
         type=MovieScene2DTransformSection
         end_frame=60
         start_frame=1
         channels [Translation.Y {type=MovieSceneScriptingFloatChannel, keys [[1,100],[60,0]]}]
        }]
      }
     ]
    }
    Button_Matchmaking {
     id=#3
     possessed_class=WBP_UI_Button_Matchmaking_C
     tracks [
      {
       type=MovieSceneFloatTrack
       display_name=Render Opacity
       sections [{type=MovieSceneFloatSection, end_frame=65, start_frame=1, channels [None {type=MovieSceneScriptingFloatChannel, keys [[1,0],[6,0],[65,1]]}]}]
      }
      {
       type=MovieScene2DTransformTrack
       sections [{
         type=MovieScene2DTransformSection
         end_frame=65
         start_frame=1
         channels [Translation.Y {type=MovieSceneScriptingFloatChannel, keys [[1,100],[6,100],[65,0]]}]
        }]
      }
     ]
    }
    Button_Close {
     id=#4
     possessed_class=WBP_UIKit_Button_Regular_C
     tracks [
      {
       type=MovieSceneFloatTrack
       display_name=Render Opacity
       sections [{type=MovieSceneFloatSection, end_frame=65, start_frame=1, channels [None {type=MovieSceneScriptingFloatChannel, keys [[1,0],[6,0],[65,1]]}]}]
      }
      {
       type=MovieScene2DTransformTrack
       sections [{
         type=MovieScene2DTransformSection
         end_frame=65
         start_frame=1
         channels [Translation.Y {type=MovieSceneScriptingFloatChannel, keys [[1,99.6631],[6,100],[65,0]]}]
        }]
      }
     ]
    }
    WBP_MatchmakingSettings_MapRotation_Entry {
     id=#5
     possessed_class=WBP_MatchmakingSettings_MapRotation_Entry_C
     tracks [
      {
       type=MovieScene2DTransformTrack
       sections [{
         type=MovieScene2DTransformSection
         end_frame=60
         start_frame=1
         channels [Translation.Y {type=MovieSceneScriptingFloatChannel, keys [[1,100],[60,0]]}]
        }]
      }
      {
       type=MovieSceneFloatTrack
       display_name=Render Opacity
       sections [{type=MovieSceneFloatSection, end_frame=60, start_frame=1, channels [None {type=MovieSceneScriptingFloatChannel, keys [[1,0],[60,1]]}]}]
      }
     ]
    }
   ]
   display_rate {numerator=60}
  }
 }]
graphs [
 BP_GetDesiredFocusTarget {nodes [
   Get Desired Focus Target()
   n1 = Get WBP_MatchmakingSettings_PolyList() -> (WBP_MatchmakingSettings_PolyList:WBP Matchmaking Settings Poly List Object Reference)
   n2 = Get ListView!(self:Fort Polymorphic List View Object Reference=@n1) -> (ListView:Common List View Object Reference)
   > Return Node!(
    execute
    ReturnValue:Widget Object Reference=GetWidgetFromItem!(
     self:Fort Polymorphic View Object Reference=@n1
     Item:Object Reference=Get Selected Item!(self:List View Object Reference=@n2) -> (ReturnValue:Object Reference) {is=title K2Node_CallFunction}
    ) -> (ReturnValue:User Widget Object Reference)
   )
  ]}
 EventGraph {nodes [
   n1 = "On Clicked (Button_Close)"!() -> (OutputDelegate, then, Button:Common Button Base Object Reference) {is=title K2Node_ComponentBoundEvent}
   DeactivateWidget(execute=[@n1.then, @n2.then])
   n2 = "On Clicked (Button_Matchmaking)"!() -> (OutputDelegate, then, Button:Common Button Base Object Reference) {is=title K2Node_ComponentBoundEvent}
   n3 = OnItemHovered!() -> (OutputDelegate, then, Item:Object Reference, bIsHovered:bool=~) {is=title K2Node_CustomEvent}
   > Branch(Condition=@n3.bIsHovered)
   > n4 = Cast To FortSettingVM!(execute, Object=@n3.Item) -> (then, CastFailed, AsFort Setting VM:Fort Setting VM Object Reference) {is=title K2Node_DynamicCast}
   n5 = Get WBP_MatchmakingSettings_Panel() -> (WBP_MatchmakingSettings_Panel:WBP Matchmaking Settings Panel Object Reference)
   > "SetText (Text)"(
    self=Get Txt_SettingTitle!(self:WBP Matchmaking Settings Panel Object Reference=@n5) -> (Txt_SettingTitle:Common Text Object Reference)
    InText=GetName!(self:Fort Setting VM Object Reference=@n4) -> (ReturnValue:Text)
   )
   > "SetText (Text)"(
    self=Get Txt_SettingDesc!(self:WBP Matchmaking Settings Panel Object Reference=@n5) -> (Txt_SettingDesc:Common Text Object Reference)
    InText=GetDescription!(self:Fort Setting VM Object Reference=@n4) -> (ReturnValue:Text)
   )
   n6 = OnListViewEntriesGenerated!() -> (OutputDelegate, then, NumEntries:int) {is=title K2Node_CustomEvent}
   > RequestRefreshFocus(self:Common Activatable Widget Object Reference)
   Event Construct()
   n7 = Get ListView(self=Get WBP_MatchmakingSettings_PolyList())
   > Bind Event to On Item Is Hovered Changed(self:List View Object Reference=@n7, Delegate=@n3) {is=title K2Node_AddDelegate}
   > Bind Event to On Entries Generated(self:List View Base Object Reference=@n7, Delegate=@n6) {is=title K2Node_AddDelegate}
   > PlayAnimation(InAnimation=Get Anim_Intro() -> (Anim_Intro:Widget Animation Object Reference))
   Event On Activated()
   > Set SelectedIslandViewModel(SelectedIslandViewModel:Selected Island View Model Object Reference=GetSelectedIslandViewModel!(self:Matchmaking UISubsystem Object Reference=Get MatchmakingUISubsystem() -> (ReturnValue:Matchmaking UISubsystem Object Reference) {is=title K2Node_GetSubsystem}) -> (ReturnValue:Selected Island View Model Object Reference)) -> (Output_Get:Selected Island View Model Object Reference)
   > Set SelectedIslandViewModel!(
    execute
    SelectedIslandViewModel=Get SelectedIslandViewModel() -> (SelectedIslandViewModel:Selected Island View Model Object Reference)
    self:WBP Matchmaking Settings Map Rotation Entry Object Reference=Get WBP_MatchmakingSettings_MapRotation_Entry() -> (WBP_MatchmakingSettings_MapRotation_Entry:WBP Matchmaking Settings Map Rotation Entry Object Reference)
   ) -> (then, Output_Get)
   > SetSelectedIndex!(execute, self:List View Object Reference=Get ListView(self=Get WBP_MatchmakingSettings_PolyList()), Index:int=0) -> (then)
   Event Destruct()
   > Unbind Event from On Item Is Hovered Changed(self:List View Object Reference=@n7, Delegate=@n3) {is=title K2Node_RemoveDelegate}
   > Unbind Event from On Entries Generated(self:List View Base Object Reference=@n7, Delegate=@n6) {is=title K2Node_RemoveDelegate}
  ]}
 HandleIslandChanged {nodes [
   HandleIslandChanged!() -> (then, Island:Online Link Id Structure) {is=title K2Node_FunctionEntry}
   n1 = Get SelectedIslandViewModel()
   > Branch(Condition=IsValid(Object=@n1))
   > LoadFullscreenImage!(
    execute
    self:Selected Island View Model Object Reference=@n1
    LocalPlayer:Local Player Object Reference=GetOwningLocalPlayer!(self:Widget Object Reference) -> (ReturnValue:Local Player Object Reference)
   ) -> (then)
   n2 = GetOwningLocalPlayer()
   n3 = Get SelectedIslandViewModel()
   > SetBadgeText!(
    execute
    self:WBP UIKit Badge Base Object Reference=Get Badge_CCU() -> (Badge_CCU:WBP UIKit Badge Object Reference)
    Text:Text="To Text (Integer)"(Value=GetCCU!(self:Selected Island View Model Object Reference=@n3, LocalPlayer:Local Player Object Reference=@n2) -> (ReturnValue:int=0))
   ) -> (then)
   > RefreshIslandLogo()
  ]}
 NavigateLeftFromPlayButton {nodes [
   NavigateLeftFromPlayButton!() -> (then, Navigation:EUINavigation Enum) {is=title K2Node_FunctionEntry}
   n1 = Get WBP_MatchmakingSettings_PolyList()
   n2 = Get ListView(self=@n1)
   > Return Node(ReturnValue=GetWidgetFromItem(self=@n1, Item=Get Selected Item(self=@n2)))
  ]}
 RefreshIslandLogo {nodes [
   RefreshIslandLogo() {is=title K2Node_FunctionEntry}
   > n1 = Branch(Condition=HasLogoWidget!(self:Selected Island View Model Object Reference=Get SelectedIslandViewModel()) -> (ReturnValue:bool))
   > LoadLogoWidget!(execute, self:Selected Island View Model Object Reference=Get SelectedIslandViewModel(), OnWidgetLoaded:"Delegate (by ref)"=Create Event()) -> (then)
   SetIslandLogo!(execute=@n1.else, self, LogoClass:User Widget Class Reference) -> (then)
  ]}
 SetIslandLogo {nodes [
   n1 = SetIslandLogo!() -> (then, LogoClass:User Widget Class Reference) {is=title K2Node_FunctionEntry}
   > n2 = Branch(Condition=IsValidClass!(Class:Object Class Reference=@n1) -> (ReturnValue:bool))
   > n3 = Create Widget!(execute:Exec, Class:User Widget Class Reference=@n1, OwningPlayer:Player Controller Object Reference=GetOwningPlayer()) -> (then:Exec, ReturnValue:User Widget Object Reference)
   > SetContent!(
    execute
    self:Content Widget Object Reference=Get NamedSlot_Logo() -> (NamedSlot_Logo:Named Slot Object Reference)
    Content:Widget Object Reference=@n3
   ) -> (then, ReturnValue:Panel Slot Object Reference)
   > SetActiveWidget(self=Get Switcher_Logo() -> (Switcher_Logo:Common Visibility Switcher Object Reference), Widget=Get NamedSlot_Logo())
   SetActiveWidget(execute=@n2.else, self=Get Switcher_Logo(), Widget=Get Txt_LogoFallback() -> (Txt_LogoFallback:Common Text Object Reference))
   > ClearChildren(self=Get NamedSlot_Logo())
  ]}
]
variables []
widgets {
 info {
  inheritedWidgetCount=0
  parentClass=/Script/MatchmakingSettingsUI.MatchmakingSettingsWidget
  rootWidgetClass=&GridPanel
  widgetCount=25
 }
 widgets [
  GP_Container {widgetClassPath=&GridPanel, widget=GP_Container, properties {bOverride_Cursor=true, columnFill [1,0], rowFill [1,0]}}
  Background {
   widgetClassPath=&Image
   bIsVariable=true
   parent=GP_Container
   widget=Background
   properties {
    visibility=SelfHitTestInvisible
    brush {
     bIsDynamicallyLoaded=false
     drawAs=Image
     imageType=FullColor
     mirroring=NoMirror
     resourceName=None
     resourceObject=/MatchmakingSettingsUI/MatchmakingSettings/MaterialInstances/MI_UI_MatchmakingSettings_Background
     tiling=NoTile
     imageSize=<32 32>
     margin=<ltrb 0 0 0 0>
     outlineSettings {
      bUseBrushTransparency=false
      roundingType=HalfHeightRadius
      width=0
      color {colorUseRule=UseColor_Specified, specifiedColor=<rgba 0 0 0 0>}
      cornerRadii=<0 0 0 0>
     }
     tintColor {colorUseRule=UseColor_Specified, specifiedColor=<rgba 1 1 1 1>}
     uVRegion {bIsValid=false, max=<0 0>, min=<0 0>}
    }
   }
   slot {class=GridSlot, properties {columnSpan=2, layer=-10, rowSpan=2}}
  }
  WBP_CaptureForPostBufferUpdate {
   widgetClassPath=&WBP_CaptureForPostBufferUpdate
   parent=GP_Container
   slot {class=GridSlot, properties {columnSpan=2, layer=-5, rowSpan=2}}
  }
  BlurOverlay {
   widgetClassPath=&Image
   parent=GP_Container
   widget=BlurOverlay
   properties {
    visibility=SelfHitTestInvisible
    brush {
     bIsDynamicallyLoaded=false
     drawAs=Image
     imageType=FullColor
     mirroring=NoMirror
     resourceName=None
     resourceObject=/MatchmakingSettingsUI/MatchmakingSettings/Materials/MI_UI_MM_Overlay_Blur
     tiling=NoTile
     imageSize=<32 32>
     margin=<ltrb 0 0 0 0>
     outlineSettings {
      bUseBrushTransparency=false
      roundingType=HalfHeightRadius
      width=0
      color {colorUseRule=UseColor_Specified, specifiedColor=<rgba 0 0 0 0>}
      cornerRadii=<0 0 0 0>
     }
     tintColor {colorUseRule=UseColor_Specified, specifiedColor=<rgba 1 1 1 1>}
     uVRegion {bIsValid=false, max=<0 0>, min=<0 0>}
    }
   }
   slot {class=GridSlot, properties {columnSpan=2, row=1}}
  }
  SZ_Content {
   widgetClassPath=&SafeZone
   parent=GP_Container
   widget=SZ_Content
   properties {padRight=false}
   slot {class=GridSlot, properties {layer=20, rowSpan=2}}
  }
  VB_Content {
   widgetClassPath=&GridPanel
   parent=SZ_Content
   widget=VB_Content
   properties {bOverride_Cursor=true, columnFill [1], rowFill [0,0,1,0]}
   slot {class=SafeZoneSlot, properties {padding=<ltrb 80 80 80 48>}}
  }
  Switcher_Logo {
   widgetClassPath=&CommonVisibilitySwitcher
   bIsVariable=true
   parent=VB_Content
   widget=Switcher_Logo
   slot {class=GridSlot, properties {horizontalAlignment=HAlign_Left, verticalAlignment=VAlign_Top}}
  }
  Txt_LogoFallback {
   widgetClassPath=&CommonTextBlock
   bIsVariable=true
   parent=Switcher_Logo
   widget=Txt_LogoFallback
   properties {
    text=Title Fallback
    wrapTextAt=696
    font {
     bForceMonospaced=false
     bMaterialIsStencil=false
     fontMaterial=null
     fontObject=&NotoSans
     letterSpacing=0
     monospacedWidth=1
     size=24
     skewAmount=0
     typefaceFontName=NotoSans-Regular
     outlineSettings {
      bApplyOutlineToDropShadows=false
      bMiteredCorners=false
      bSeparateFillAlpha=false
      outlineMaterial=null
      outlineSize=0
      outlineColor=<rgba 0 0 0 1>
     }
    }
    shadowOffset=<0 0>
   }
   slot {class=CommonVisibilitySwitcherSlot, properties {horizontalAlignment=HAlign_Left, verticalAlignment=VAlign_Top}}
  }
  NamedSlot_Logo {
   widgetClassPath=&NamedSlot
   bIsVariable=true
   parent=Switcher_Logo
   widget=NamedSlot_Logo
   slot {class=CommonVisibilitySwitcherSlot}
  }
  HorizontalBox_0 {
   widgetClassPath=&HorizontalBox
   parent=VB_Content
   widget=HorizontalBox_0
   slot {class=GridSlot, properties {horizontalAlignment=HAlign_Left, row=1, verticalAlignment=VAlign_Top, padding=<ltrb 0 16 0 0>}}
  }
  Badge_CCU {
   widgetClassPath=/FortUILibrary/UIKit/Badge/WBP_UIKit_Badge
   bIsVariable=true
   parent=HorizontalBox_0
   widget=Badge_CCU
   properties {badgeSize=Default, badgeText=$CCU, badgeType=Text, indicatorConfig=None}
   slot {class=HorizontalBoxSlot, properties {horizontalAlignment=HAlign_Left, verticalAlignment=VAlign_Center}}
  }
  Txt_PlayingNow {
   widgetClassPath=&CommonTextBlock
   parent=HorizontalBox_0
   widget=Txt_PlayingNow
   properties {
    text=Playing Now
    colorAndOpacity {colorUseRule=UseColor_Specified, specifiedColor=<rgba 1 1 1 0.7>}
    font {
     bForceMonospaced=false
     bMaterialIsStencil=false
     fontMaterial=null
     fontObject=&NotoSans
     letterSpacing=0
     monospacedWidth=1
     size=24
     skewAmount=0
     typefaceFontName=NotoSans-Regular
     outlineSettings {
      bApplyOutlineToDropShadows=false
      bMiteredCorners=false
      bSeparateFillAlpha=false
      outlineMaterial=null
      outlineSize=0
      outlineColor=<rgba 0 0 0 1>
     }
    }
    shadowOffset=<0 0>
   }
   slot {class=HorizontalBoxSlot, properties {horizontalAlignment=HAlign_Left, verticalAlignment=VAlign_Center, padding=<ltrb 8 0 0 0>}}
  }
  O_ForTopPadding {
   widgetClassPath=&Overlay
   parent=VB_Content
   widget=O_ForTopPadding
   slot {class=GridSlot, properties {layer=10, row=2, padding=<ltrb 0 64 0 -16>}}
  }
  O_ListClipping {
   widgetClassPath=&Overlay
   parent=O_ForTopPadding
   widget=O_ListClipping
   properties {clipping=ClipToBounds}
   slot {class=OverlaySlot, properties {horizontalAlignment=HAlign_Fill, verticalAlignment=VAlign_Bottom, padding=<ltrb -16 0 -16 0>}}
  }
  WBP_MatchmakingSettings_PolyList {
   widgetClassPath=&/WBP_MatchmakingSettings_PolyList
   bIsVariable=true
   parent=O_ListClipping
   widget=WBP_MatchmakingSettings_PolyList
   properties {navigation=&self:WidgetTree.WBP_MatchmakingSettings_PolyList.WidgetNavigation_0}
   slot {class=OverlaySlot, properties {horizontalAlignment=HAlign_Fill, verticalAlignment=VAlign_Fill, padding=<ltrb 16 16 16 16>}}
  }
  WBP_MatchmakingSettings_MapRotation_Entry {
   widgetClassPath=/MatchmakingSettingsUI/MatchmakingSettings/SettingEntries/MapRotation/WBP_MatchmakingSettings_MapRotation_Entry
   bIsVariable=true
   parent=VB_Content
   widget=WBP_MatchmakingSettings_MapRotation_Entry
   slot {class=GridSlot, properties {row=3, padding=<ltrb 0 24 0 0>}}
  }
  SZ_Panel {
   widgetClassPath=&SafeZone
   parent=GP_Container
   widget=SZ_Panel
   properties {padLeft=false}
   slot {class=GridSlot, properties {column=1, layer=10, row=1, verticalAlignment=VAlign_Bottom}}
  }
  B_PanelContent {
   widgetClassPath=&CommonBorder
   parent=SZ_Panel
   widget=B_PanelContent
   properties {
    background {
     bIsDynamicallyLoaded=false
     drawAs=NoDrawType
     imageType=FullColor
     mirroring=NoMirror
     resourceName=None
     resourceObject=/MatchmakingSettingsUI/MatchmakingSettings/MaterialInstances/MI_UI_MatchmakingSettings_PanelOverlay
     tiling=NoTile
     imageSize=<32 32>
     margin=<ltrb 0 0 0 0>
     outlineSettings {
      bUseBrushTransparency=false
      roundingType=HalfHeightRadius
      width=0
      color {colorUseRule=UseColor_Specified, specifiedColor=<rgba 0 0 0 0>}
      cornerRadii=<0 0 0 0>
     }
     tintColor {colorUseRule=UseColor_Specified, specifiedColor=<rgba 1 1 1 1>}
     uVRegion {bIsValid=false, max=<0 0>, min=<0 0>}
    }
    padding=<ltrb 0 0 0 0>
   }
   slot {class=SafeZoneSlot, properties {vAlign=VAlign_Bottom, padding=<ltrb 0 0 80 48>}}
  }
  GP_SidePanel {
   widgetClassPath=&GridPanel
   parent=B_PanelContent
   widget=GP_SidePanel
   properties {bOverride_Cursor=true, columnFill [2,1], rowFill [1,0]}
   slot {class=BorderSlot, properties {padding=<ltrb 0 0 0 0>}}
  }
  SB_Content {
   widgetClassPath=&SizeBox
   parent=GP_SidePanel
   widget=SB_Content
   properties {bOverride_WidthOverride=true, widthOverride=624}
   slot {class=GridSlot, properties {columnSpan=2, padding=<ltrb 0 0 0 48>}}
  }
  ScrollBox_Content {
   widgetClassPath=&ScrollBox
   parent=SB_Content
   widget=ScrollBox_Content
   properties {
    bAnimateWheelScrolling=true
    scrollAnimationInterpolationSpeed=12
    scrollbarPadding=<ltrb 12 0 0 0>
    widgetBarStyle {
     thickness=12
     draggedThumbImage {
      bIsDynamicallyLoaded=false
      drawAs=RoundedBox
      imageType=FullColor
      mirroring=NoMirror
      resourceName=../../../Engine/Content/Slate/Common/Scrollbar_Thumb.png
      resourceObject=null
      tiling=NoTile
      imageSize=<12 12>
      margin=<ltrb 0.25 0.25 0.25 0.25>
      outlineSettings {
       bUseBrushTransparency=false
       roundingType=FixedRadius
       width=2
       color {colorUseRule=UseColor_Specified, specifiedColor=<rgba 0 0 0 0>}
       cornerRadii=<6 6 6 6>
      }
      tintColor {colorUseRule=UseColor_Specified, specifiedColor=<rgba 1 1 1 1>}
      uVRegion {bIsValid=false, max=<0 0>, min=<0 0>}
     }
     horizontalBackgroundImage {
      bIsDynamicallyLoaded=false
      drawAs=Box
      imageType=NoImage
      mirroring=NoMirror
      resourceName=None
      resourceObject=/FortUILibrary/UIKit/ScrollBar/MaterialInstances/MI_UIKit_ScrollbarBackground
      tiling=NoTile
      imageSize=<12 12>
      margin=<ltrb 0.5 0.5 0.5 0.5>
      outlineSettings {
       bUseBrushTransparency=false
       roundingType=HalfHeightRadius
       width=0
       color {colorUseRule=UseColor_Specified, specifiedColor=<rgba 0 0 0 0>}
       cornerRadii=<0 0 0 0>
      }
      tintColor {colorUseRule=UseColor_Specified, specifiedColor=<rgba 1 1 1 1>}
      uVRegion {bIsValid=false, max=<0 0>, min=<0 0>}
     }
     horizontalBottomSlotImage {
      bIsDynamicallyLoaded=false
      drawAs=NoDrawType
      imageType=FullColor
      mirroring=NoMirror
      resourceName=../../../Engine/Content/Slate/Common/Scrollbar_Background_Horizontal.png
      resourceObject=null
      tiling=NoTile
      imageSize=<8 8>
      margin=<ltrb 0 0 0 0>
      outlineSettings {
       bUseBrushTransparency=false
       roundingType=HalfHeightRadius
       width=0
       color {colorUseRule=UseColor_Specified, specifiedColor=<rgba 0 0 0 0>}
       cornerRadii=<0 0 0 0>
      }
      tintColor {colorUseRule=UseColor_Specified, specifiedColor=<rgba 1 1 1 1>}
      uVRegion {bIsValid=false, max=<0 0>, min=<0 0>}
     }
     horizontalTopSlotImage {
      bIsDynamicallyLoaded=false
      drawAs=NoDrawType
      imageType=FullColor
      mirroring=NoMirror
      resourceName=../../../Engine/Content/Slate/Common/Scrollbar_Background_Horizontal.png
      resourceObject=null
      tiling=NoTile
      imageSize=<8 8>
      margin=<ltrb 0 0 0 0>
      outlineSettings {
       bUseBrushTransparency=false
       roundingType=HalfHeightRadius
       width=0
       color {colorUseRule=UseColor_Specified, specifiedColor=<rgba 0 0 0 0>}
       cornerRadii=<0 0 0 0>
      }
      tintColor {colorUseRule=UseColor_Specified, specifiedColor=<rgba 1 1 1 1>}
      uVRegion {bIsValid=false, max=<0 0>, min=<0 0>}
     }
     hoveredThumbImage {
      bIsDynamicallyLoaded=false
      drawAs=RoundedBox
      imageType=FullColor
      mirroring=NoMirror
      resourceName=../../../Engine/Content/Slate/Common/Scrollbar_Thumb.png
      resourceObject=null
      tiling=NoTile
      imageSize=<12 12>
      margin=<ltrb 0.25 0.25 0.25 0.25>
      outlineSettings {
       bUseBrushTransparency=false
       roundingType=FixedRadius
       width=2
       color {colorUseRule=UseColor_Specified, specifiedColor=<rgba 0 0 0 0>}
       cornerRadii=<6 6 6 6>
      }
      tintColor {colorUseRule=UseColor_Specified, specifiedColor=<rgba 1 1 1 1>}
      uVRegion {bIsValid=false, max=<0 0>, min=<0 0>}
     }
     normalThumbImage {
      bIsDynamicallyLoaded=false
      drawAs=RoundedBox
      imageType=FullColor
      mirroring=NoMirror
      resourceName=../../../Engine/Content/Slate/Common/Scrollbar_Thumb.png
      resourceObject=null
      tiling=NoTile
      imageSize=<12 12>
      margin=<ltrb 0.25 0.25 0.25 0.25>
      outlineSettings {
       bUseBrushTransparency=false
       roundingType=FixedRadius
       width=2
       color {colorUseRule=UseColor_Specified, specifiedColor=<rgba 0 0 0 0>}
       cornerRadii=<6 6 6 6>
      }
      tintColor {colorUseRule=UseColor_Specified, specifiedColor=<rgba 1 1 1 0.75>}
      uVRegion {bIsValid=false, max=<0 0>, min=<0 0>}
     }
     verticalBackgroundImage {
      bIsDynamicallyLoaded=false
      drawAs=Box
      imageType=NoImage
      mirroring=NoMirror
      resourceName=None
      resourceObject=&MI_UIKit_ScrollbarBackground
      tiling=NoTile
      imageSize=<12 12>
      margin=<ltrb 0.5 0.5 0.5 0.5>
      outlineSettings {
       bUseBrushTransparency=false
       roundingType=HalfHeightRadius
       width=0
       color {colorUseRule=UseColor_Specified, specifiedColor=<rgba 0 0 0 0>}
       cornerRadii=<0 0 0 0>
      }
      tintColor {colorUseRule=UseColor_Specified, specifiedColor=<rgba 1 1 1 1>}
      uVRegion {bIsValid=false, max=<0 0>, min=<0 0>}
     }
     verticalBottomSlotImage {
      bIsDynamicallyLoaded=false
      drawAs=NoDrawType
      imageType=FullColor
      mirroring=NoMirror
      resourceName=../../../Engine/Content/Slate/Common/Scrollbar_Background_Vertical.png
      resourceObject=null
      tiling=NoTile
      imageSize=<8 8>
      margin=<ltrb 0 0 0 0>
      outlineSettings {
       bUseBrushTransparency=false
       roundingType=HalfHeightRadius
       width=0
       color {colorUseRule=UseColor_Specified, specifiedColor=<rgba 0 0 0 0>}
       cornerRadii=<0 0 0 0>
      }
      tintColor {colorUseRule=UseColor_Specified, specifiedColor=<rgba 1 1 1 1>}
      uVRegion {bIsValid=false, max=<0 0>, min=<0 0>}
     }
     verticalTopSlotImage {
      bIsDynamicallyLoaded=false
      drawAs=NoDrawType
      imageType=FullColor
      mirroring=NoMirror
      resourceName=../../../Engine/Content/Slate/Common/Scrollbar_Background_Vertical.png
      resourceObject=null
      tiling=NoTile
      imageSize=<8 8>
      margin=<ltrb 0 0 0 0>
      outlineSettings {
       bUseBrushTransparency=false
       roundingType=HalfHeightRadius
       width=0
       color {colorUseRule=UseColor_Specified, specifiedColor=<rgba 0 0 0 0>}
       cornerRadii=<0 0 0 0>
      }
      tintColor {colorUseRule=UseColor_Specified, specifiedColor=<rgba 1 1 1 1>}
      uVRegion {bIsValid=false, max=<0 0>, min=<0 0>}
     }
    }
    widgetStyle {
     barThickness=12
     bottomShadowBrush {
      bIsDynamicallyLoaded=false
      drawAs=Image
      imageType=FullColor
      mirroring=NoMirror
      resourceName=None
      resourceObject=/MatchmakingSettingsUI/MatchmakingSettings/MaterialInstances/MI_UI_MatchmakingSettings_PanelOverlay_Bottom
      tiling=NoTile
      imageSize=<16 8>
      margin=<ltrb 0.5 0 0.5 1>
      outlineSettings {
       bUseBrushTransparency=false
       roundingType=HalfHeightRadius
       width=0
       color {colorUseRule=UseColor_Specified, specifiedColor=<rgba 0 0 0 0>}
       cornerRadii=<0 0 0 0>
      }
      tintColor {colorUseRule=UseColor_Specified, specifiedColor=<rgba 1 1 1 1>}
      uVRegion {bIsValid=false, max=<0 0>, min=<0 0>}
     }
     horizontalScrolledContentPadding=<ltrb 0 0 1 0>
     leftShadowBrush {
      bIsDynamicallyLoaded=false
      drawAs=NoDrawType
      imageType=FullColor
      mirroring=NoMirror
      resourceName=../../../Engine/Content/Slate/Common/ScrollBoxShadowLeft.png
      resourceObject=null
      tiling=NoTile
      imageSize=<8 16>
      margin=<ltrb 1 0.5 0 0.5>
      outlineSettings {
       bUseBrushTransparency=false
       roundingType=HalfHeightRadius
       width=0
       color {colorUseRule=UseColor_Specified, specifiedColor=<rgba 0 0 0 0>}
       cornerRadii=<0 0 0 0>
      }
      tintColor {colorUseRule=UseColor_Specified, specifiedColor=<rgba 1 1 1 1>}
      uVRegion {bIsValid=false, max=<0 0>, min=<0 0>}
     }
     rightShadowBrush {
      bIsDynamicallyLoaded=false
      drawAs=NoDrawType
      imageType=FullColor
      mirroring=NoMirror
      resourceName=../../../Engine/Content/Slate/Common/ScrollBoxShadowRight.png
      resourceObject=null
      tiling=NoTile
      imageSize=<8 16>
      margin=<ltrb 0 0.5 1 0.5>
      outlineSettings {
       bUseBrushTransparency=false
       roundingType=HalfHeightRadius
       width=0
       color {colorUseRule=UseColor_Specified, specifiedColor=<rgba 0 0 0 0>}
       cornerRadii=<0 0 0 0>
      }
      tintColor {colorUseRule=UseColor_Specified, specifiedColor=<rgba 1 1 1 1>}
      uVRegion {bIsValid=false, max=<0 0>, min=<0 0>}
     }
     topShadowBrush {
      bIsDynamicallyLoaded=false
      drawAs=Image
      imageType=FullColor
      mirroring=NoMirror
      resourceName=None
      resourceObject=/MatchmakingSettingsUI/MatchmakingSettings/MaterialInstances/MI_UI_MatchmakingSettings_PanelOverlay_Top
      tiling=NoTile
      imageSize=<16 8>
      margin=<ltrb 0.5 1 0.5 0>
      outlineSettings {
       bUseBrushTransparency=false
       roundingType=HalfHeightRadius
       width=0
       color {colorUseRule=UseColor_Specified, specifiedColor=<rgba 0 0 0 0>}
       cornerRadii=<0 0 0 0>
      }
      tintColor {colorUseRule=UseColor_Specified, specifiedColor=<rgba 1 1 1 1>}
      uVRegion {bIsValid=false, max=<0 0>, min=<0 0>}
     }
     verticalScrolledContentPadding=<ltrb 0 0 0 1>
    }
   }
   slot {class=SizeBoxSlot}
  }
  WBP_MatchmakingSettings_Panel {
   widgetClassPath=&/WBP_MatchmakingSettings_Panel
   bIsVariable=true
   parent=ScrollBox_Content
   widget=WBP_MatchmakingSettings_Panel
   slot {class=ScrollBoxSlot, properties {verticalAlignment=VAlign_Bottom, size {sizeRule=Fill, value=1}}}
  }
  WBP_CaptureForPostBufferUpdate_1 {
   widgetClassPath=&WBP_CaptureForPostBufferUpdate
   parent=GP_SidePanel
   widget=WBP_CaptureForPostBufferUpdate_1
   slot {class=GridSlot, properties {columnSpan=2, layer=-5, row=1}}
  }
  Button_Matchmaking {
   widgetClassPath=&/WBP_UI_Button_Matchmaking
   bIsVariable=true
   parent=GP_SidePanel
   widget=Button_Matchmaking
   properties {
    navigation=&self:WidgetTree.Button_Matchmaking.WidgetNavigation_1
    triggeringInputAction {dataTable=&InputActionDataTable, rowName=Play}
   }
   slot {class=GridSlot, properties {row=1}}
  }
  Button_Close {
   widgetClassPath=&WBP_UIKit_Button_Regular
   bIsVariable=true
   parent=GP_SidePanel
   widget=Button_Close
   properties {
    bHideInputAction=true
    buttonType=Quiet
    mobileMinWidth=240
    text=Close
    triggeringInputAction {dataTable=&InputActionDataTable, rowName=Close}
   }
   slot {class=GridSlot, properties {column=1, row=1, padding=<ltrb 24 0 0 0>}}
  }
 ]
}
