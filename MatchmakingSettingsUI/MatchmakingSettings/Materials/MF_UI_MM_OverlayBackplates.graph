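# UI material function that serves as a shared palette for overlay backplates in the matchmaking settings UI.
# It takes no inputs and exposes Color and Opacity output pairs for five variants: A, B, B Weak, C and Disabled.
# Each output is fed by its own hardcoded scalar constant, so the colors are grayscale.
# The A and B Weak opacities are left at zero.
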MaterialFunction /MatchmakingSettingsUI/MatchmakingSettings/Materials/MF_UI_MM_OverlayBackplates
graph {
 properties {previewMaterialDomain=MD_UI}
 nodes [
  FunctionOutput(None=Constant() {r=0.25}) {outputName=A Color}
  FunctionOutput(None=Constant()) {outputName=A Opacity, sortPriority=1}
  FunctionOutput(None=Constant() {r=0.25}) {outputName=B Color, sortPriority=10}
  FunctionOutput(None=Constant() {r=0.4}) {outputName=B Opacity, sortPriority=11}
  FunctionOutput(None=Constant() {r=0.02}) {outputName=B Weak Color, sortPriority=20}
  FunctionOutput(None=Constant()) {outputName=B Weak Opacity, sortPriority=21}
  FunctionOutput(None=Constant() {r=0.0025}) {outputName=C Color, sortPriority=30}
  FunctionOutput(None=Constant() {r=0.3333}) {outputName=C Opacity, sortPriority=31}
  FunctionOutput(None=Constant() {r=0.25}) {outputName=Disabled Color, sortPriority=40}
  FunctionOutput(None=Constant() {r=0.9}) {outputName=Disabled Opacity, sortPriority=41}
 ]
}
